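; Flying Orb actor for DragonCart narrative events that periodically scans for players and unopened chests, retaliates with a laser, and repels intruders with a shield.
; Calendar events initialize scan and laser on the server, with cheat overrides and a sequencer lockout.
; Damage or random targeting enters timed shooting mode, pausing pings while focus and yaw checks drive laser charge.
; Event-router messages move it above Dark Voyager Echo bosses and target their instigators.
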(event Custom|SetOrbZ
  (params (OrbZ :type Float%20%28double-precision%29))
  (bind _returnvalue_location (Math|Transform|BreakTransform (Transformation|GetActorTransform)))
  ;[0.03803 0.1406 0.03308 0.5] Cheats
  (Transformation|SetWorldLocation (Variables|Default|GetDefaultSceneRoot) (Math|Vector|MakeVector (.x _returnvalue_location) (.y _returnvalue_location) OrbZ) false true))

(event Custom|DCN_HideOrb
  ;[0.03803 0.1406 0.03308 0.5] Cheats
  (Variables|Default|SetShowOrb false))

(event Custom|Init%20Things%20After%20Orb%20Moves
  ;[0.3769 0.7396 0.7255 0.5] Begin Play
  (CallFunction|InitEyeLoopAudio))

(event Custom|OnDiedEvent
  (params (DamagedActor :type Actor%20Object%20Reference) (Damage :type Float%20%28single-precision%29) (InstigatedBy :type Controller%20Object%20Reference) (DamageCauser :type Actor%20Object%20Reference) (HitLocation :type Vector) (FHitComponent :type Primitive%20Component%20Object%20Reference) (BoneName :type Name) (Momentum :type Vector))
  ;[0.03803 0.1406 0.03308 0.5] Orb Destroyed - Debug
  (Fortnite|FortShippingLog self (Utilities|String|Append "Instigated By:
" (Utilities|GetDisplayName InstigatedBy) "Damage Causer:
" (Utilities|GetDisplayName DamageCauser) "Damage Amount:
" (Utilities|String|ToString(Float) Damage) "
BGA_DCN_OrbActor")))

(event Custom|DCN_ShowOrb
  ;[0.03803 0.1406 0.03308 0.5] Cheats
  (Variables|Default|SetShowOrb true))

(event Custom|Disable%20Gameplay%20During%20Sequencer
  ;[0.4247 0.1616 0.4427 0.5] Sequencer Helper Functions
  (Variables|Default|SetDisable%20Orb%20Gameplay true) ; don't let the Orb do things while it's being\nanimated from sequencer such as during\nthe Orb rift in
)

(event Custom|Init%20Eye%20Loop%20Audio
  ;[0.3769 0.7396 0.7255 0.5] Begin Play
  (Variables|Default|SetStartingLocation (Transformation|GetWorldLocation (Variables|Default|GetAudio%20-%20Eye%20Loop))))

(event Custom|Bind%20Mid%20Match%20Calendar%20Events
  ;[0.4427 0.3669 0.2154 0.5] Mid Match Calendar Events (Laser/Scan)
  (if (Networking|IsServer)
    (Utilities|FlowControl|Sequence
      (:then_0
        (MCPContext|WatchCalendarEvent (Variables|Default|GetCalendar%20-%20Scan%20Controller)
          (:OnEventActive
            (CallFunction|InitializeScanPing))) ; Scan Calendar Event
      )
      (:then_1
        (MCPContext|WatchCalendarEvent (Variables|Default|GetCalendar%20-%20Laser%20Controller)
          (:OnEventActive
            (CallFunction|InitializeLaser))) ; Laser Calendar Event
      )) ; 0 - Scan\n1 - Laser
  ))

(event Custom|SetOrbScale
  (params (OrbScale :type Float%20%28double-precision%29))
  ;[0.03803 0.1406 0.03308 0.5] Cheats
  (Transformation|SetWorldScale3D (Variables|Default|GetDefaultSceneRoot) (Math|Conversions|ToVector(Float) OrbScale)))

(event Custom|Enable%20Gameplay%20After%20Sequencer
  ;[0.4247 0.1616 0.4427 0.5] Sequencer Helper Functions
  (Utilities|FlowControl|Delay 2.0) ; ..wait for the sequence to end and the\ntransform track to no longer be controlling\nthe orb so the "turn and shoot" can do its \nthing
  (Variables|Default|SetDisable%20Orb%20Gameplay false) ; re-enable gameplay after sequencer\nsuch as the rift in
  (CallFunction|InitializeScanPing)
  (CallFunction|InitializeLaser)
  (Utilities|FlowControl|Delay 0.5) ; let the Orb initialize.. probably\nnot necessary, but adding it anyhow
  (CallFunction|ShootAtRandomTarget))

(event Custom|Initialize%20Scan%20Ping
  ;[0.3769 0.7396 0.7255 0.5] Begin Play
  ;; Scan / Ping Init
  (CallFunction|UpdatePingScanEnabledState)
  (CallFunction|UpdatePingTimers))

(event Custom|Update%20Ping%20Timers
  ;[0.3769 0.7396 0.7255 0.5] Begin Play
  ;; Scan / Ping Init
  (if (Variables|Default|GetPingsEnabled)
    (CallFunction|RestartPingTimer)
    (else
      (CallFunction|StopPingTimer)))
  (if (Variables|Default|GetShootingEnabled)
    (CallFunction|ResetPOIPing)))

(event EventBeginPlay
  (bind _bga_health (Variables|Default|GetBGA%20Health))
  ;[0.3769 0.7396 0.7255 0.5] Begin Play
  (CallFunction|LogIt "Begin Play. Waiting for Playlist Ready..")
  (Class|FortAsyncActionCurrentPlaylistReady|PlaylistDataReadyAsync
    (:OnReady
      (CallFunction|LogIt "Waiting for Calendar Sync Completed.")
      (MCPContext|InitialCalendarSyncCompleted
        (:OnSyncComplete
          (CallFunction|LogIt "Calendar Sync Complete!")
          (Utilities|FlowControl|Sequence
            (:then_0
              (Health|SetMaxHealthDontCreateASC _bga_health) ; Set Orb as Invulnerable at start
              (Health|SetHealth _bga_health))
            (:then_1
              (CallFunction|BindMidMatchCalendarEvents)
              (CallFunction|InitializeLaser)
              (CallFunction|InitializeScanPing))
            (:then_2
              (Networking|CanExecuteCosmeticEvents
                (:True
                  (CallFunction|AdjustLoopingAudioVolume))))
            (:then_3
              (EventDispatchers|BindEventtoOnDied (EventDispatchers|CreateEvent :!event Custom|OnDiedEvent))
              (Game|BindEventtoOnDestroyed (EventDispatchers|CreateEvent :!event Custom|EventOnDestroy))
              (CallFunction|HandleDarkVoyagerEchoEvents))))))))

(event Custom|DC_TeleportPlayerToOrb
  (bind (_returnvalue_x _returnvalue_y _returnvalue_z) (Math|Vector|BreakVector (Transformation|GetActorLocation)))
  ;[0.03803 0.1406 0.03308 0.5] Cheats
  (Transformation|SetActorLocation (Game|GetPlayerPawn 0) (Math|Vector|MakeVector _returnvalue_x _returnvalue_y (+ _returnvalue_z 7e+03)) false false) ; teleport player to Orb
)

(event Custom|DC_EnableScan
  ;[0.03803 0.1406 0.03308 0.5] Cheats
  (Variables|Default|SetScan%20-%20Force%20Enable%20fromCheat true)
  (CallFunction|InitializeScanPing))

(event EventDestroyed
  ;[0.03803 0.1406 0.03308 0.5] Orb Destroyed - Debug
  (|CallParentFunction)
  (CallFunction|StopPingTimer))

(event Custom|DC_EnableShooting
  ;[0.03803 0.1406 0.03308 0.5] Cheats
  (Variables|Default|SetShootingEnabled true)
  (Variables|Default|SetLaser%20-%20Force%20Enable%20from%20Cheat true)
  (CallFunction|InitializeLaser))

(event Custom|EventOnDestroy
  (params (DestroyedActor :type Actor%20Object%20Reference))
  ;[0.03803 0.1406 0.03308 0.5] Orb Destroyed - Debug
  (Fortnite|FortShippingLog self (Utilities|String|Append "Destroyed Actor:
" (Utilities|GetDisplayName DestroyedActor))))

(event Custom|DCN_EnableShooting
  ;[0.03803 0.1406 0.03308 0.5] Cheats
  (Variables|Default|SetShootingEnabled true)
  (Variables|Default|SetLaser%20-%20Force%20Enable%20from%20Cheat true)
  (CallFunction|InitializeLaser))

(event Custom|DC_TeleportPlayerToEcho
  (bind (_returnvalue_x _returnvalue_y _returnvalue_z) (Math|Vector|BreakVector (Transformation|GetActorLocation (Variables|DarkVoyagerEcho|GetNearby%20Dark%20Voyager%20Echo%20Boss))))
  ;[0.03803 0.1406 0.03308 0.5] Cheats
  (Transformation|SetActorLocation (Game|GetPlayerPawn 0) (Math|Vector|MakeVector _returnvalue_x _returnvalue_y (+ _returnvalue_z 2e+03)) false false) ; teleport player to echo
)

(event Custom|Initialize%20Laser
  ;[0.3769 0.7396 0.7255 0.5] Begin Play
  ;; Laser Init
  (CallFunction|UpdateLaserEnabledState)
  (CallFunction|UpdatePingTimers) ; Ping Timers get updated if shooting\nis enabled, so go refresh Ping Timers.
)

(event Custom|DCN_DisableShooting
  ;[0.03803 0.1406 0.03308 0.5] Cheats
  (Variables|Default|SetShootingEnabled false)
  (Variables|Default|SetLaser%20-%20Force%20Enable%20from%20Cheat false))

(event OnComponentBeginOverlap(ShieldImpulse)
  (params (OverlappedComponent :type Primitive%20Component%20Object%20Reference) (OtherActor :type Actor%20Object%20Reference) (OtherComp :type Primitive%20Component%20Object%20Reference) (OtherBodyIndex :type Integer) (bFromSweep :type Boolean) (SweepResult :type Hit%20Result%20Structure%20%28by%20ref%29))
  (bind (_bblockinghit _binitialoverlap _time _distance _location) (Collision|BreakHitResult SweepResult))
  (bind _self self)
  ;[0.002327 0.3458 0.4688] Shield Knockback
  (Networking|SwitchHasAuthority
    (:Authority
      (bind _asfort_player_pawn_athena (Utilities|Casting|CastToFortPlayerPawnAthena OtherActor)
        (:then
          (if (not (GameplayTags|HasMatchingGameplayTag _asfort_player_pawn_athena "(TagName=\"Pawn.Athena.NPC\")"))
            (Variables|Default|SetOverlapResultStruct SweepResult)
            (bind _outhits (Collision|MultiCapsuleTraceByChannel :Start (Transformation|GetActorLocation OtherActor) :End (+ (Transformation|GetActorLocation OtherActor) (* (Math|Vector|GetUnitDirection(Vector) (Transformation|GetActorLocation OtherActor) _location) (ScalableFloat|GetValueatLevel (Variables|Default|GetDestroyDistance)))) :Radius 8e+01 :HalfHeight 1.5e+02 :TraceChannel "TraceTypeQuery11" :bTraceComplex false :DrawDebugType "ForDuration" :bIgnoreSelf true :TraceColor "(R=1.000000,G=0.000000,B=0.000000,A=1.000000)" :TraceHitColor "(R=0.000000,G=1.000000,B=0.000000,A=1.000000)" :DrawTime 2.0))
            (for _array_element _outhits
              (if (not (Math|Boolean|ANDBoolean (Variables|/Script/Engine.Actor|GetbCanBeDamaged (.HitActor (Collision|BreakHitResult _array_element))) (Utilities|IsA (.HitActor (Collision|BreakHitResult _array_element)) "/Script/FortniteGame.BuildingActor") (Utilities|IsA (.HitActor (Collision|BreakHitResult _array_element)) "/Script/FortniteGame.BuildingTrap")))
                (GameplayEffects|ApplyGameplayEffectToTarget :self (Ability|GetAbilitySystemComponent _asfort_player_pawn_athena) :GameplayEffectClass "/DragonCartNarrative/Gameplay/FlyingOrb/Abilities/GE_DCN_Orb_Knockback_Damage.GE_DCN_Orb_Knockback_Damage_C" :Target (Ability|GetAbilitySystemComponent (.HitActor (Collision|BreakHitResult _array_element))) :Level 0.0)))
            (Utilities|FlowControl|DelayUntilNextTick)
            (GameplayEffects|FortApplyGameplayEffectToSelf :self (Ability|GetFortAbilitySystemComponent _asfort_player_pawn_athena) :GameplayEffectClass (Variables|Default|GetGE_Knockback) :OptionalSourceObject _self :OptionalEffectCauser _self :Level 0.0)
            (GameplayEffects|FortApplyGameplayEffectToSelf :self (Ability|GetFortAbilitySystemComponent _asfort_player_pawn_athena) :GameplayEffectClass "/Game/Athena/Items/Parents/FallImmunity/GE_Athena_TillLandFallDamageImmunity_NoFX_Grant_ApplyInAirOnly.GE_Athena_TillLandFallDamageImmunity_NoFX_Grant_ApplyInAirOnly_C" :OptionalSourceObject _self :OptionalEffectCauser _self :Level 0.0)))
        (:CastFailed
          (bind _asfort_pickup (Utilities|Casting|CastToFortPickup OtherActor)
            (:then
              (bind (_returnvalue_roll _returnvalue_pitch _returnvalue_yaw) (Math|Rotator|BreakRotator (Math|Rotator|FindLookatRotation (Transformation|GetActorLocation) (Transformation|GetActorLocation _asfort_pickup))))
              (Pickup|TossPickupinDirection _asfort_pickup (* (Math|Vector|MakeVector _returnvalue_roll _returnvalue_pitch (Math|Float|Absolute(Float) _returnvalue_yaw)) 5.0) 0 false true))
            (:CastFailed
              (bind _asfort_athena_vehicle (Utilities|Casting|CastToFortAthenaVehicle OtherActor))
              (CallFunction|KnockbackVehicle _asfort_athena_vehicle)))))
      (bind _returnvalue (Utilities|Time|SetTimerbyEvent (EventDispatchers|CreateEvent :!event Custom|ScanForOverlap) (ScalableFloat|GetValueatLevelwithContext (Variables|Default|GetRetryLoopTime)) true))
      (Variables|Default|SetOverlapCheckTimer _returnvalue))))

(event Custom|DC_Fire
  ;[0.4427 0.1223 0.1011 0.5] Laser - Enter/Exit shooting mode
  (CallFunction|SelectValidTarget)
  (if (Variables|Default|GetPingsEnabled_WhenShootingLaser)
    (CallFunction|PingPlayersinArea)))

(event Custom|SetOrbModeVisuals
  (params (AttackMode :type Boolean))
  ;[0.4427 0.1223 0.1011 0.5] Laser - Enter/Exit shooting mode
  (if AttackMode
    (bind (_direction _idle_attack) (Timeline|OrbAttacking PlayFromStart)
      (:Update
        (Rendering|Material|SetScalarParameterValue "/DragonCartNarrative/Assets/VFX/RiftPortalIdle/MPC_DragonCart_Orb_Rift.MPC_DragonCart_Orb_Rift" "Orb Attack Mode" _idle_attack :!class /Script/Engine.KismetMaterialLibrary)))
    (else
      (Utilities|FlowControl|Delay (ScalableFloat|GetValueatLevel (Variables|Default|GetHF_DelayBeforeOrbCanShoot)))
      (bind (_direction_2 _idle_attack_2) (Timeline|OrbCooldown PlayFromStart)
        (:Update
          (Rendering|Material|SetScalarParameterValue "/DragonCartNarrative/Assets/VFX/RiftPortalIdle/MPC_DragonCart_Orb_Rift.MPC_DragonCart_Orb_Rift" "Orb Attack Mode" _idle_attack_2 :!class /Script/Engine.KismetMaterialLibrary))
        (:Finished
          (if (Variables|Default|GetPingsEnabled)
            (CallFunction|RestartPingTimer)
            (else
              _))
          (Variables|Default|SetShootingCooldown false))))))

(event Custom|InitiateWhizby
  ;[0.4427 0.1223 0.1011 0.5] Laser - Handle Audio Whizby
  (Networking|CanExecuteCosmeticEvents
    (:True
      (Utilities|FlowControl|Delay (ScalableFloat|GetValueatLevel (Variables|Default|GetOrb_FireTime)))
      (bind _returnvalue (Audio|SpawnSoundatLocation "/DragonCartNarrative/Assets/Audio/Laser/MSS_DragonCart_Laser_Prj.MSS_DragonCart_Laser_Prj" (Transformation|GetSocketLocation (Variables|Default|GetStaticMesh) "EyeSocket")))
      (Variables|Default|SetWhizbyAudio _returnvalue)
      (bind (_direction _newtrack_0) (Timeline|WhizbyAudio_Time PlayFromStart)
        (:Update
          (bind _whizbyaudio (Variables|Default|GetWhizbyAudio)
            (:then
              (Transformation|SetWorldLocation _whizbyaudio (Math|Vector|Lerp(Vector) (Transformation|GetSocketLocation (Variables|Default|GetStaticMesh) "EyeSocket") (Variables|Default|GetFinalHit) _newtrack_0) true false))
            (:else)))))))

(event Custom|PingPlayersInArea
  (bind _visualpingradius (Variables|Default|GetVisualPingRadius))
  (bind _returnvalue (ScalableFloat|GetValueatLevel _visualpingradius))
  (bind _staticmesh (Variables|Default|GetStaticMesh))
  (bind _returnvalue_2 (Transformation|GetSocketLocation _staticmesh "EyeSocket"))
  ;[0.1237 0.2274 0.8125 0.5] Ping Players in the area every x seconds
  (CallFunction|LogIt "PingPlayersInArea")
  (Rendering|Debug|DrawDebugSphere _returnvalue_2 _returnvalue 12 "(R=1.000000,G=0.000000,B=0.006200,A=1.000000)" 3e+01 1e+01)
  (Utilities|FlowControl|Sequence
    (:then_0
      ;; Find nearby pawns.
      ;; Find nearby chests (only those that are not open yet).
      (bind _pawnswithinnormalradius (CallFunction|FindPawnsWithinPingDistance))
      (bind _chestsfound (CallFunction|FindOpenChestsWithinPingDistance)))
    (:then_1
      ;; Ping (human) players
      (for (_array_element _array_index) _pawnswithinnormalradius
        (bind _asfort_player_pawn (Utilities|Casting|CastToFortPlayerPawn _array_element))
        (bind _asfort_player_controller_athena (Utilities|Casting|CastToFortPlayerControllerAthena (Pawn|GetController _asfort_player_pawn)))
        (CallFunction|PingPlayer _asfort_player_controller_athena _returnvalue_2 _returnvalue _pawnswithinnormalradius _chestsfound (Math|Conversions|ToFloat(Integer) _array_index)) ; It is a player. Ping them.
      ))
    (:then_2
      ;; Reveal the Orb - Play ping visual and audio effects
      (CallFunction|RevealOrb))))

(event Custom|Retaliate
  ;[0.4427 0.1223 0.1011 0.5] Laser - Enter/Exit shooting mode
  (CallFunction|SelectValidTarget)
  (if (Variables|Default|GetPingsEnabled_WhenShootingLaser)
    (CallFunction|PingPlayersinArea)))

(event Custom|DC_Ping
  (bind _visualpingradius (Variables|Default|GetVisualPingRadius))
  (bind _returnvalue (ScalableFloat|GetValueatLevel _visualpingradius))
  (bind _staticmesh (Variables|Default|GetStaticMesh))
  (bind _returnvalue_2 (Transformation|GetSocketLocation _staticmesh "EyeSocket"))
  ;[0.1237 0.2274 0.8125 0.5] Ping Players in the area every x seconds
  (CallFunction|LogIt "PingPlayersInArea")
  (Rendering|Debug|DrawDebugSphere _returnvalue_2 _returnvalue 12 "(R=1.000000,G=0.000000,B=0.006200,A=1.000000)" 3e+01 1e+01)
  (Utilities|FlowControl|Sequence
    (:then_0
      ;; Find nearby pawns.
      ;; Find nearby chests (only those that are not open yet).
      (bind _pawnswithinnormalradius (CallFunction|FindPawnsWithinPingDistance))
      (bind _chestsfound (CallFunction|FindOpenChestsWithinPingDistance)))
    (:then_1
      ;; Ping (human) players
      (for (_array_element _array_index) _pawnswithinnormalradius
        (bind _asfort_player_pawn (Utilities|Casting|CastToFortPlayerPawn _array_element))
        (bind _asfort_player_controller_athena (Utilities|Casting|CastToFortPlayerControllerAthena (Pawn|GetController _asfort_player_pawn)))
        (CallFunction|PingPlayer _asfort_player_controller_athena _returnvalue_2 _returnvalue _pawnswithinnormalradius _chestsfound (Math|Conversions|ToFloat(Integer) _array_index)) ; It is a player. Ping them.
      ))
    (:then_2
      ;; Reveal the Orb - Play ping visual and audio effects
      (CallFunction|RevealOrb))))

(event Custom|DC_Scan
  (bind _visualpingradius (Variables|Default|GetVisualPingRadius))
  (bind _returnvalue (ScalableFloat|GetValueatLevel _visualpingradius))
  (bind _staticmesh (Variables|Default|GetStaticMesh))
  (bind _returnvalue_2 (Transformation|GetSocketLocation _staticmesh "EyeSocket"))
  ;[0.1237 0.2274 0.8125 0.5] Ping Players in the area every x seconds
  (CallFunction|LogIt "PingPlayersInArea")
  (Rendering|Debug|DrawDebugSphere _returnvalue_2 _returnvalue 12 "(R=1.000000,G=0.000000,B=0.006200,A=1.000000)" 3e+01 1e+01)
  (Utilities|FlowControl|Sequence
    (:then_0
      ;; Find nearby pawns.
      ;; Find nearby chests (only those that are not open yet).
      (bind _pawnswithinnormalradius (CallFunction|FindPawnsWithinPingDistance))
      (bind _chestsfound (CallFunction|FindOpenChestsWithinPingDistance)))
    (:then_1
      ;; Ping (human) players
      (for (_array_element _array_index) _pawnswithinnormalradius
        (bind _asfort_player_pawn (Utilities|Casting|CastToFortPlayerPawn _array_element))
        (bind _asfort_player_controller_athena (Utilities|Casting|CastToFortPlayerControllerAthena (Pawn|GetController _asfort_player_pawn)))
        (CallFunction|PingPlayer _asfort_player_controller_athena _returnvalue_2 _returnvalue _pawnswithinnormalradius _chestsfound (Math|Conversions|ToFloat(Integer) _array_index)) ; It is a player. Ping them.
      ))
    (:then_2
      ;; Reveal the Orb - Play ping visual and audio effects
      (CallFunction|RevealOrb))))

(event Custom|DC_StopFocus
  ;[0.4427 0.1223 0.1011 0.5] Laser - Enter/Exit shooting mode
  (Development|PrintString "Shoot Mode Stopped ViaCheat")
  (Utilities|Time|ClearandInvalidateTimerbyHandle (Variables|ShootingMode|GetShootingModeTimerHandle))
  (Variables|ShootingMode|SetInShootingMode false)
  (Class|FortActorFocusComponent|ClearFocus (Variables|Default|GetFortActorFocus))
  (Utilities|Time|ClearandInvalidateTimerbyHandle (Variables|ShootingMode|GetAngleCheckTimerHandle)))

(event Custom|ShootingModeTimeLimit
  ;[0.4427 0.1223 0.1011 0.5] Laser - Enter/Exit shooting mode
  (Development|PrintString "Shoot Mode Time Limit Reached!")
  (Utilities|Time|ClearandInvalidateTimerbyHandle (Variables|ShootingMode|GetShootingModeTimerHandle))
  (Variables|ShootingMode|SetInShootingMode false)
  (Class|FortActorFocusComponent|ClearFocus (Variables|Default|GetFortActorFocus))
  (Utilities|Time|ClearandInvalidateTimerbyHandle (Variables|ShootingMode|GetAngleCheckTimerHandle)))

(event Custom|Handle%20Dark%20Voyager%20Echo%20Events
  ;[0.4427 0.2195 0 0.5] Dark Voyager Echo boss
  (if (Networking|IsServer)
    (Utilities|FlowControl|Sequence
      (:then_0
        ;; Orb over Echo Boss Calendar Event
        (MCPContext|InitialCalendarSyncCompleted
          (:OnSyncComplete
            (bind _biseventactive (CalendarEvents|GetCalendarEventInformation "DC_OrbOverEcho"))
            (if _biseventactive
              (CallFunction|LogIt "Dark Voyager Echo boss calendar event - ACTIVE")
              (Variables|DarkVoyagerEcho|SetOrb%20is%20Aggro true)
              (CallFunction|WatchForDarkVoyagerEchoBossEvents)
              (else
                (CallFunction|LogIt "Dark Voyager Echo boss calendar event - inactive"))))) ; Wait for calendar sync.
      )
      (:then_1
        ;; DC_Orb - start from cheat
        (bind (_asynctaskproxy _payload) (GameplayEventRouter|StartListeningToStatefulEvent (GameplayEventRouter|GetGlobalEventRouter) "/DragonCartNarrative/Gameplay/EventRouters/DCN_Orb_Cheats_EventRouter.DCN_Orb_Cheats_EventRouter")
          (:OnEventReceived
            _)
          (:OnSavedEventStateExists
            _)) ; handle cheats\n\n"cheat ke * DC_Orb" .. this comes from BP_DCN_CalendarDataLayer
        (switch Utilities|FlowControl|Switch|SwitchonDCN_Orb_Cheats_Eventrouter_Enum (Utilities|Struct|BreakDCNOrbCheatsEventRouter _payload)
          (:NewEnumerator0
            (CallFunction|LogIt "DC_Orb event router cheat received.")
            (CallFunction|DCEnableShooting) ; normally enabled through calendar\nevent but override that for the cheat
            (CallFunction|WatchForDarkVoyagerEchoBossEvents) ; from cheat, this usually only gets called from calendar
            (Utilities|FlowControl|Delay 1.0) ; wait for startup logic.. just giving\nPIE a chance as watching for the event\nrouter events will also probably move\nthe Orb.. but we're going to force it\nagain from the cheat so the Orb can\nbe cycled to different spots. ..so using\nthe cheat will move the Orb twice on\nthe first use. Give a second for that\nto happen.
            (Variables|DarkVoyagerEcho|SetOrb%20is%20Aggro true)
            (CallFunction|GetRandomBossEchoAndMoveOrb)))))))

(event Custom|UpdateOrbRotationAudio
  (bind (_returnvalue_roll _returnvalue_pitch _returnvalue_yaw) (Math|Rotator|BreakRotator (Transformation|GetActorRotation self)))
  ;[0.4427 0.1223 0.1011 0.5] Laser - Orb Rotation Audio Handling
  (Networking|CanExecuteCosmeticEvents
    (:True
      (if (Variables|ShootingMode|GetInShootingMode)
        (Variables|Default|GetOrbRotationAudio
          (:then
            _)
          (:else
            (bind _returnvalue (Audio|SpawnSoundAttached "/DragonCartNarrative/Assets/Audio/Laser/MSS_DragonCart_Orb_Targeting_Rotate_LP.MSS_DragonCart_Orb_Targeting_Rotate_LP" (Variables|Default|GetStaticMesh)))
            (bind _output_get (Variables|Default|SetOrbRotationAudio _returnvalue))))
        (Audio|Parameter|SetFloatParameter _output_get "Rotation" (* (* _returnvalue_roll _returnvalue_pitch) _returnvalue_yaw) :!class /Script/AudioExtensions.AudioParameterControllerInterface)
        (Audio|Parameter|ExecuteTriggerParameter _output_get "CheckRotation")
        (else
          (bind _orbrotationaudio (Variables|Default|GetOrbRotationAudio)
            (:then
              (Audio|Components|Audio|FadeOut _orbrotationaudio 1.0 0.0 "Linear" :!class /Script/Engine.AudioComponent))
            (:else)))))))

(event Custom|ShootingModeStartedAudio
  ;[0.4427 0.1223 0.1011 0.5] Laser - Enter/Exit shooting mode
  ;; Audio
  (Networking|CanExecuteCosmeticEvents
    (:True
      (if (Variables|ShootingMode|GetInShootingMode)
        _)
      (Utilities|FlowControl|DoOnce false
        (:Completed
          (Audio|SpawnSoundAttached "/DragonCartNarrative/Assets/Audio/Laser/MSS_DragonCart_Laser_FireModeActivated_Orb.MSS_DragonCart_Laser_FireModeActivated_Orb" (Variables|Default|GetStaticMesh)))))))

(event Custom|ScanForOverlap
  (bind _shieldimpulse (Variables|Default|GetShieldImpulse))
  (bind _overlappingactors (Collision|GetOverlappingActors _shieldimpulse 0 :!class /Script/Engine.PrimitiveComponent))
  ;[0.003471 0.444 0.5938 0.5] Shield - Repeat Player Knock Back
  (if (> (Utilities|Array|Length _overlappingactors) 0)
    (for _array_element _overlappingactors
      (CallFunction|RetryKnockback _array_element))
    (else
      (Utilities|Time|ClearandInvalidateTimerbyHandle (Variables|Default|GetOverlapCheckTimer)))))

(event Custom|PlayKnockbackGCN
  (params (OverlapResultStruct :type Hit%20Result%20Structure))
  (bind (_bblockinghit _binitialoverlap _time _distance _location _impactpoint _normal _impactnormal _physmat _hitactor) (Collision|BreakHitResult OverlapResultStruct))
  ;[0.003471 0.444 0.5938 0.5] Shield - Player GCN
  (Utilities|FlowControl|DoOnce false
    (:Completed
      (Utilities|FlowControl|Sequence
        (:then_0
          (Utilities|FlowControl|Delay 0.25))
        (:then_1
          _))
      (CallFunction|TriggerRipple _location 4e+03 2.0)
      (GameplayCue|ExecuteGameplayCue%28Burst%29Local _hitactor (Variables|Default|GetKnockbackGCTag) (Ability|GameplayCue|MakeGameplayCueParameters :NormalizedMagnitude 0.0 :RawMagnitude 0.0 :Location _location :Normal (* _normal -1.0) :GameplayEffectLevel 1 :AbilityLevel 1 :bReplicateLocationWhenUsingMinimalRepProxy false)))))

(event Custom|RestartPingTimer
  (bind _hf_scanretriggertime (Variables|Default|GetHF_ScanRetriggerTime))
  (bind _returnvalue (ScalableFloat|GetValueatLevel _hf_scanretriggertime))
  ;[0.1237 0.2274 0.8125 0.5] Ping Auto-Timer
  (bind _returnvalue_2 (Utilities|Time|SetTimerbyEvent (EventDispatchers|CreateEvent :!event Custom|PingPlayersInArea) _returnvalue true))
  (Variables|Default|SetTH_OrbPing _returnvalue_2)
  (CallFunction|LogIt (Utilities|String|Append "Ping timer restarted. Set to fire every " (Utilities|String|ToString(Float) _returnvalue) " seconds.")))

(event Custom|RetryKnockback
  (params (ActorToKnock :type Actor%20Object%20Reference))
  (bind _self self)
  ;[0.003471 0.444 0.5938 0.5] Shield - Repeat Player Knock Back
  (bind _asfort_player_pawn_athena (Utilities|Casting|CastToFortPlayerPawnAthena ActorToKnock)
    (:then
      (Utilities|IsValid (Ability|GetFortAbilitySystemComponent _asfort_player_pawn_athena)
        (:Is%20Valid
          (GameplayEffects|FortApplyGameplayEffectToSelf :self (Ability|GetFortAbilitySystemComponent _asfort_player_pawn_athena) :GameplayEffectClass "/DragonCartNarrative/Gameplay/FlyingOrb/Effects/GE_OrbKnockback.GE_OrbKnockback_C" :OptionalSourceObject _self :OptionalEffectCauser _self :Level 0.0)
          (GameplayEffects|FortApplyGameplayEffectToSelf :self (Ability|GetFortAbilitySystemComponent _asfort_player_pawn_athena) :GameplayEffectClass "/Game/Athena/Items/Parents/FallImmunity/GE_Athena_TillLandFallDamageImmunity_NoFX_Grant_ApplyInAirOnly.GE_Athena_TillLandFallDamageImmunity_NoFX_Grant_ApplyInAirOnly_C" :OptionalSourceObject _self :OptionalEffectCauser _self :Level 0.0))))
    (:CastFailed
      (bind _asfort_athena_vehicle (Utilities|Casting|CastToFortAthenaVehicle ActorToKnock))
      (CallFunction|KnockbackVehicle _asfort_athena_vehicle))))

(event Custom|StartCheckAngle
  (params (TargetActor :type Actor%20Object%20Reference))
  ;[0.4427 0.1223 0.1011 0.5] Laser - Check angle if Orb eye can see target
  (Variables|ShootingMode|SetTargetedActor TargetActor)
  (bind _returnvalue (Utilities|Time|SetTimerbyEvent (EventDispatchers|CreateEvent :!event Custom|CheckPlayerAngle) 0.25 true))
  (Variables|ShootingMode|SetAngleCheckTimerHandle _returnvalue))

(event Custom|ResetLaserCharge
  ;[0.4427 0.1223 0.1011 0.5] Laser - Check angle if Orb eye can see target
  (CallFunction|ShootingModeTimeLimit)
  (Utilities|FlowControl|DoOnce false
    (:Completed
      (CallFunction|OrbLaserCharge (Variables|ShootingMode|GetTargetedActor))
      (CallFunction|UpdateOrbRotationAudio))))

(event Custom|Get%20Random%20Boss%20Echo%20and%20Move%20Orb
  (bind _hf__orb_height_over_echo (Variables|DarkVoyagerEcho|GetHF_%20Orb%20Height%20Over%20Echo))
  (bind _returnvalue (ScalableFloat|GetValueatLevel _hf__orb_height_over_echo))
  (bind _nearby_dark_voyager_echo_boss (Variables|DarkVoyagerEcho|GetNearby%20Dark%20Voyager%20Echo%20Boss))
  (bind (_returnvalue_x _returnvalue_y _returnvalue_z) (Math|Vector|BreakVector (Transformation|GetActorLocation _nearby_dark_voyager_echo_boss)))
  (bind _returnvalue_2 (+ _returnvalue_z _returnvalue))
  (bind _returnvalue_3 (Math|Vector|MakeVector _returnvalue_x _returnvalue_y _returnvalue_2))
  ;[0.4427 0.2195 0 0.5] Dark Voyager Echo boss
  ;; Get Random Echo Boss and Move Orb
  (CallFunction|LogIt "Get Random Boss Echo and Move Orb")
  (bind (_echo_boss _found_boss) (CallFunction|GetRandomEchoBoss))
  (if _found_boss
    (Variables|DarkVoyagerEcho|SetNearby%20Dark%20Voyager%20Echo%20Boss _echo_boss)
    (CallFunction|LogIt "Have valid Echo Boss")
    ;;; Move Orb over Dark Voyager Echo
    (Transformation|SetActorLocation _returnvalue_3) ; Move Orb to be above the Dark Voyager Echo
    (Rendering|SetActorHiddenInGame false) ; unhide the Orb after moving it\n\n..the OFPA actor has hidden in game\nchecked on it.
    (CallFunction|LogIt (Utilities|String|Append "Moving Orb over echo at: " (Utilities|String|ToString(Vector) _returnvalue_3)))
    (CallFunction|InitThingsAfterOrbMoves)
    (else
      (CallFunction|LogIt "Didn't find valid echo boss!" true))))

(event Custom|ResetPOIPing
  ;[0.4427 0.1223 0.1011 0.5] Laser - Reset Shooting Mode
  (Ability|Async|WaitGameplayEventToActor 0 "(TagName=\"Event.DragonCart.Orb.BeamBlast.Charge.Reset\")"
    (:then
      (Ability|Async|WaitGameplayEventToActor 0 "(TagName=\"Event.DragonCart.Orb.BeamBlast.Charge.ReadyToFire\")"
        (:EventReceived
          (Class|FortActorFocusComponent|ClearFocus (Variables|Default|GetFortActorFocus))
          (Variables|ShootingMode|SetTargetedActor 0))))
    (:EventReceived
      (Variables|Default|SetShootingCooldown true)
      (CallFunction|ResetLaserCharge))))

(event Custom|StopPingTimer
  ;[0.1237 0.2274 0.8125 0.5] Ping Auto-Timer
  (Utilities|Time|ClearandInvalidateTimerbyHandle (Variables|Default|GetTH_OrbPing))
  (CallFunction|LogIt "Ping timer stopped."))

(event Custom|CheckPlayerAngle
  (bind _targetedactor (Variables|ShootingMode|GetTargetedActor))
  (bind _returnvalue (Transformation|GetActorLocation _targetedactor))
  (bind _self self)
  (bind _returnvalue_2 (Transformation|GetActorTransform _self))
  (bind _returnvalue_3 (Math|Rotator|FindRelativeLookatRotation _returnvalue_2 _returnvalue))
  (bind _orbtargetyawmax (Variables|ShootingMode|GetOrbTargetYawMax))
  (bind _returnvalue_4 (ScalableFloat|GetValueatLevelwithContext _orbtargetyawmax))
  (bind _orbtargetyawmin (Variables|ShootingMode|GetOrbTargetYawMin))
  (bind _returnvalue_5 (ScalableFloat|GetValueatLevelwithContext _orbtargetyawmin))
  (bind _yaw (.yaw _returnvalue_3))
  (bind _returnvalue_6 (Math|Float|InRange(Float) _yaw _returnvalue_5 _returnvalue_4))
  ;[0.4427 0.1223 0.1011 0.5] Laser - Check angle if Orb eye can see target
  (Utilities|IsValid _targetedactor
    (:Is%20Valid
      (Class|FortActorFocusComponent|SetActorFocus (Variables|Default|GetFortActorFocus) _targetedactor)
      (CallFunction|UpdateOrbRotationAudio)
      (if _returnvalue_6
        (Development|PrintString "I see you!")
        (Variables|ShootingMode|SetTargetInArc _returnvalue_6)
        (Utilities|FlowControl|DoOnce false
          (:Completed
            (CallFunction|OrbLaserCharge (Variables|ShootingMode|GetTargetedActor))
            (CallFunction|UpdateOrbRotationAudio)))
        (else
          (Development|PrintString "Get back here!")
          (Variables|ShootingMode|SetTargetInArc _returnvalue_6))))))

(event Custom|SetLaserFinalHitLocation
  (params (FinalHit :type Vector) (Target :type Self%20Object%20Reference))
  ;[0.4427 0.1223 0.1011 0.5] Laser - Final Hit Location
  (Variables|Default|SetFinalHit FinalHit)
  (CallFunction|InitiateWhizby Target))

(event EventOnDamagePlayEffects
  (params (Damage :type Float%20%28single-precision%29) (DamageTags :type Gameplay%20Tag%20Container%20Structure%20%28by%20ref%29) (Momentum :type Vector :default "0, 0, 0") (HitInfo :type Hit%20Result%20Structure%20%28by%20ref%29) (InstigatedBy :type Fort%20Pawn%20Object%20Reference) (DamageCauser :type Actor%20Object%20Reference) (EffectContext :type Gameplay%20Effect%20Context%20Handle%20Structure))
  ;[0.003471 0.444 0.5938 0.5] Shield - NDC on Shot
  (CallFunction|TriggerNDCImpact HitInfo Damage)
  (Development|PrintString "Triggered NDC On Weapon Hit")
  (Development|PrintString (Utilities|String|Append "I've been shot by: " (Utilities|GetDisplayName InstigatedBy) " and they are at: " (Utilities|String|ToString(Vector) (Transformation|GetActorLocation InstigatedBy)) " and they dealt a damage total of: " (Utilities|String|ToString(Float) Damage)) true true "(R=0.000000,G=0.660000,B=1.000000,A=1.000000)" 2e+01))

(event Custom|Watch%20for%20Dark%20Voyager%20Echo%20Boss%20Events
  ;[0.4427 0.2195 0 0.5] Dark Voyager Echo boss
  ;; Orb Aggro Events from Echo Boss
  (bind (_asynctaskproxy _payload) (GameplayEventRouter|StartListeningToStatefulEvent (GameplayEventRouter|GetGlobalEventRouter) "/DragonCartNarrative/Gameplay/EventRouters/DCN_SpacePlunge_EventRouter.DCN_SpacePlunge_EventRouter")
    (:OnEventReceived
      _)
    (:OnSavedEventStateExists
      _))
  (switch Utilities|FlowControl|Switch|SwitchonDCN_SpacePlunge_EventRouter_Enum (Utilities|Struct|BreakDCNSpacePlungeEventRouter _payload)
    (:NewEnumerator0
      (Utilities|FlowControl|DoOnce false
        (:Completed
          (Utilities|FlowControl|Delay 0.1) ; Delay. I don't like delays, but if one\nboss spawns, the rest are probably\nspawning within this time.
          (CallFunction|GetRandomBossEchoAndMoveOrb))) ; Once a space plunge (Dark Voyager Echo boss)\nis spawned, wait a moment and then go find a\nrandom boss in the map.
    )
    (:NewEnumerator1
      _)
    (:NewEnumerator2
      _)
    (:NewEnumerator3
      _)
    (:NewEnumerator4
      _)
    (:NewEnumerator5
      _)
    (:NewEnumerator6
      _))
  (if (== (Utilities|ResolveSoftReference (.SpacePlungeactor_14_EFCA546D4A92395F158B74BE973FFF25 (Utilities|Struct|BreakDCNSpacePlungeEventRouter _payload))) (Variables|DarkVoyagerEcho|GetNearby%20Dark%20Voyager%20Echo%20Boss))
    (CallFunction|LogIt (Utilities|String|Append "shooting specific player because: " (Utilities|String|EnumtoString (Utilities|Struct|BreakDCNSpacePlungeEventRouter _payload))))
    (CallFunction|ShootSpecificPlayer (Utilities|ResolveSoftReference (.otheractor_15_36B2C9A843D950F7BAD971880037A8D3 (Utilities|Struct|BreakDCNSpacePlungeEventRouter _payload)))) ; Shoot the player that caused the Dark Voyager Echo event (ex: damaged boss, damaged drone).\nIf no player instigator (ex: target lost), the Orb will shoot a random target just to add some spice.
  ))

(event Custom|Shoot%20Specific%20Player
  (params (target%20player :type Actor%20Object%20Reference))
  ;[0.4427 0.1223 0.1011 0.5] Laser - Turn and fire when shot by player
  (if (CallFunction|CanShoot?)
    (Variables|ShootingMode|SetInShootingMode true)
    (bind _asfort_player_pawn (Utilities|Casting|CastToFortPlayerPawn target%20player)
      (:then
        (Variables|Default|SetInstigatingPlayer _asfort_player_pawn) ; shoot this actor!
      )
      (:CastFailed
        (Variables|Default|SetInstigatingPlayer 0) ; NULL\n..if this isn't set, a random target\nin the area will be shot.
      )) ; Cast Failed - shoot a random player
    (CallFunction|StopPingTimer)
    (CallFunction|Retaliate)
    (bind _returnvalue (Utilities|Time|SetTimerbyEvent (EventDispatchers|CreateEvent :!event Custom|ShootingModeTimeLimit) (ScalableFloat|GetValueatLevelwithContext (Variables|ShootingMode|GetShootingTimer))))
    (Variables|ShootingMode|SetShootingModeTimerHandle _returnvalue)))

(event Custom|Shoot%20at%20Random%20Target
  ;[0.4427 0.1223 0.1011 0.5] Laser - Turn and fire when shot by player
  (if (CallFunction|CanShoot?)
    (Variables|ShootingMode|SetInShootingMode true)
    (Variables|Default|SetInstigatingPlayer 0) ; NULL\n..if this isn't set, a random target\nin the area will be shot.
    (CallFunction|StopPingTimer)
    (CallFunction|Retaliate)
    (bind _returnvalue (Utilities|Time|SetTimerbyEvent (EventDispatchers|CreateEvent :!event Custom|ShootingModeTimeLimit) (ScalableFloat|GetValueatLevelwithContext (Variables|ShootingMode|GetShootingTimer))))
    (Variables|ShootingMode|SetShootingModeTimerHandle _returnvalue)))

(event EventOnDamageServer
  (params (Damage :type Float%20%28single-precision%29) (DamageTags :type Gameplay%20Tag%20Container%20Structure%20%28by%20ref%29) (Momentum :type Vector :default "0, 0, 0") (HitInfo :type Hit%20Result%20Structure%20%28by%20ref%29) (InstigatedBy :type Controller%20Object%20Reference) (DamageCauser :type Actor%20Object%20Reference) (EffectContext :type Gameplay%20Effect%20Context%20Handle%20Structure))
  (bind _asfort_player_controller (Utilities|Casting|CastToFortPlayerController InstigatedBy))
  ;[0.4427 0.1223 0.1011 0.5] Laser - Turn and fire when shot by player
  (|CallParentFunction Damage DamageTags Momentum HitInfo InstigatedBy DamageCauser EffectContext)
  (if (CallFunction|CanShoot?)
    (Utilities|IsValid _asfort_player_controller
      (:Is%20Valid
        (Variables|ShootingMode|SetInShootingMode true)
        (Variables|Default|SetInstigatingPlayer (Game|Player|GetPlayerPawn _asfort_player_controller)) ; player who shot the Orb
        (CallFunction|StopPingTimer)
        (CallFunction|Retaliate)
        (bind _returnvalue (Utilities|Time|SetTimerbyEvent (EventDispatchers|CreateEvent :!event Custom|ShootingModeTimeLimit) (ScalableFloat|GetValueatLevelwithContext (Variables|ShootingMode|GetShootingTimer))))
        (Variables|ShootingMode|SetShootingModeTimerHandle _returnvalue))) ; player controller valid?
  ))

(event Custom|Orb%20Changed%20Aggro%20State%20%28Client%29
  ;[0.4427 0.2195 0 0.5] Dark Voyager Echo boss
  ;; Client / Server Events for Aggro Orb VFX & Audio
  (if (not (Networking|IsServer))
    (if (Variables|DarkVoyagerEcho|GetOrb%20is%20Aggro)
      _)))

(event Custom|Orb%20Changed%20Aggro%20State%20%28Server%29
  ;[0.4427 0.2195 0 0.5] Dark Voyager Echo boss
  ;; Client / Server Events for Aggro Orb VFX & Audio
  (if (Networking|IsServer)
    (if (Variables|DarkVoyagerEcho|GetOrb%20is%20Aggro)
      _)))

(fn AdjustLoopingAudioVolume
  (bind _loopingaudiovolume (Variables|Default|GetLoopingAudioVolume))
  (Audio|Components|Audio|AdjustVolume (Variables|Default|GetAudio%20-%20Engine%20Loop) 0.5 _loopingaudiovolume "Linear" :!class /Script/Engine.AudioComponent)
  (Audio|Components|Audio|AdjustVolume (Variables|Default|GetAudio%20-%20Eye%20Loop) 0.5 _loopingaudiovolume "Linear" :!class /Script/Engine.AudioComponent))

(fn CanShoot?
  (return (Math|Boolean|ANDBoolean (Variables|Default|GetShootingEnabled) (not (Variables|ShootingMode|GetInShootingMode)) (not (Variables|Default|GetShootingCooldown)) (not (Variables|Default|GetDisable%20Orb%20Gameplay)))))

(fn FindOpenChestsWithinPingDistance
  ; Find all the ABuildingContainer actors.
  ; Then, select only those that are chests, not opened, and within squared distance.
  (bind _outactors (Actor|GetAllActorsOfClass "/Script/FortniteGame.BuildingContainer"))
  (LocalVariables|SetPingDistanceSquared (* (ScalableFloat|GetValueatLevel (Variables|Default|GetVisualPingRadius)) (ScalableFloat|GetValueatLevel (Variables|Default|GetVisualPingRadius))))
  (for _array_element _outactors
    (if (< (Transformation|GetSquaredDistanceTo 0) (LocalVariables|GetPingDistanceSquared))
      (Utilities|IsValid (Ability|GetAbilitySystemComponent _array_element)
        (:Is%20Valid
          (if (GameplayTags|HasMatchingGameplayTag (Ability|GetAbilitySystemComponent _array_element) "(TagName=\"Building.Type.Container.TreasureChest\")")
            (bind _asbuilding_container (Utilities|Casting|CastToBuildingContainer _array_element))
            (if (not (Container|IsAlreadySearched _asbuilding_container))
              (Utilities|Array|Add (LocalVariables|GetChests) _asbuilding_container) ; add non searched container
            )))) ; building container has a valid\nability system?
    ))
  (CallFunction|LogIt (Utilities|String|Append "FindOpenChestsWithinPingDistance - Found " (Utilities|String|ToString(Integer) (Utilities|Array|Length (LocalVariables|GetChests))) " chests"))
  (return (LocalVariables|GetChests)))

(fn FindPawnsWithinPingDistance
  (LocalVariables|SetNormalPingDistanceSquared (* (ScalableFloat|GetValueatLevel (Variables|Default|GetVisualPingRadius)) (ScalableFloat|GetValueatLevel (Variables|Default|GetVisualPingRadius))))
  (LocalVariables|SetPingOrigin (Transformation|GetSocketLocation (Variables|Default|GetStaticMesh) "EyeSocket"))
  (for _array_element (Variables|/Script/Engine.GameStateBase|GetPlayerArray (Fortnite|Athena|GameState|GetGameStateAthena))
    ; Add the player to CloserPawns if it is within the ping radius
    (bind _asfort_pawn (Utilities|Casting|CastToFortPawn (PlayerState|GetPawn _array_element)))
    (LocalVariables|SetPlayerDistanceSquared (Math|Vector|DistanceSquared(Vector) (Transformation|GetActorLocation _asfort_pawn) (LocalVariables|GetPingOrigin)))
    (if (<= (LocalVariables|GetPlayerDistanceSquared) (LocalVariables|GetNormalPingDistanceSquared))
      (Utilities|Array|Add (LocalVariables|GetCloserPawns) _asfort_pawn)))
  (Fortnite|FortShippingLog self (Utilities|String|ToString(Text) (Utilities|Text|FormatText "NSLOCTEXT(\"[2CCF3747C18808667CEF527EAE3EBD85]\", \"4492E86544BCC04D7FB37AB856D7F199\", \"BGA_DCN_OrbActor: FindPawnsWithinPingDistance - Found {0}pawns\")" (Utilities|Array|Length (LocalVariables|GetCloserPawns)))))
  (return (LocalVariables|GetCloserPawns)))

(fn GetRandomEchoBoss
  ; Find Random Echo Boss
  (bind _outactors (Actor|GetAllActorsOfClass "/DragonCartNarrative/Gameplay/DarkVoyager/BP_DarkVoyagerEcho_Helper.BP_DarkVoyagerEcho_Helper_C")) ; The Dark Voyager Echo boss spawns and destroys\nthis actor and passes itself as an Acto reference.\n\nBP_PlayerPawn_PAVB_TeaCake_SpacePlunge
  (bind _output_get (LocalVariables|Setecho%20boss%20helper%20actor%20%28local%29 (Utilities|Array|RandomArrayItem _outactors)))
  (Utilities|IsValid _output_get
    (:Is%20Valid
      (Utilities|IsValid (Class|BPDarkVoyagerEchoHelper|GetDarkVoyagerEchoPawnActor (LocalVariables|Getecho%20boss%20helper%20actor%20%28local%29))
        (:Is%20Valid
          (Variables|DarkVoyagerEcho|SetNearby%20Dark%20Voyager%20Echo%20Boss (Class|BPDarkVoyagerEchoHelper|GetDarkVoyagerEchoPawnActor (LocalVariables|Getecho%20boss%20helper%20actor%20%28local%29)))
          (CallFunction|LogIt (Utilities|String|Append "Found " (Utilities|String|ToString(Integer) (Utilities|Array|Length _outactors)) " Dark Voyager Echo bosses.
Chose the boss at: " (Utilities|String|ToString(Vector) (Transformation|GetActorLocation (Variables|DarkVoyagerEcho|GetNearby%20Dark%20Voyager%20Echo%20Boss)))))
          (return (Class|BPDarkVoyagerEchoHelper|GetDarkVoyagerEchoPawnActor (LocalVariables|Getecho%20boss%20helper%20actor%20%28local%29)) true))
        (:Is%20Not%20Valid
          (CallFunction|LogIt "Found the echo boss helper actor, but the helper actor didn't have a valid actor reference to the echo boss!" true)
          (return 0 false))) ; helper actor had a valid dark\nvoyager echo boss?
    )
    (:Is%20Not%20Valid
      (CallFunction|LogIt "didn't find echo boss helper actor" true)
      (return 0 false))) ; found a helper actor?
)

(fn GetMarkerDisplayTeam
  (bind _as_fort_player_controller_athena (Variables|Default|GetAs%20Fort%20Player%20Controller%20Athena))
  (bind _returnvalue (Game|GetInstigator _as_fort_player_controller_athena))
  (Utilities|IsValid _as_fort_player_controller_athena
    (:Is%20Valid
      (Utilities|IsValid _returnvalue
        (:Is%20Valid
          (return (Team|GetActorTeam _returnvalue)))
        (:Is%20Not%20Valid
          _)))
    (:Is%20Not%20Valid
      _))
  (return 255))

(fn GetVehicleImpulseValue
  (params (Actor :type Actor%20Object%20Reference))
  (bind (_returnvalue_x _returnvalue_y) (Math|Vector|BreakVector (- (Transformation|GetActorLocation Actor) (Transformation|GetWorldLocation (Variables|Default|GetShieldImpulse)))))
  (return (+ (* (Math|Vector|Normalize (Math|Vector|MakeVector _returnvalue_x _returnvalue_y)) (Math|Vector|MakeVector (ScalableFloat|GetValueatLevelwithContext (Variables|Default|GetVehicle%20Knockback%20-%20Horz)) (ScalableFloat|GetValueatLevelwithContext (Variables|Default|GetVehicle%20Knockback%20-%20Horz)))) (Math|Vector|MakeVector 0.0 0.0 (ScalableFloat|GetValueatLevelwithContext (Variables|Default|GetVehicle%20Knockback%20-%20Vert))))))

(fn HandlePlayerAbilitySet
  (params (FortPlayerControllerAthena :type Fort%20Player%20Controller%20Athena%20Object%20Reference))
  (bind _fortplayercontrollerathena (Variables|GetFortPlayerControllerAthena))
  (bind _as_radiusmarking_map (Variables|Default|GetAS_RadiusMarking_Map))
  (bind (_value _returnvalue) (Utilities|Map|Find _as_radiusmarking_map _fortplayercontrollerathena))
  (Utilities|FlowControl|Sequence
    (:then_0
      (if (Ability|GameplayAbility|HasAnyAbilitieswithAssetTag (Ability|GetAbilitySystemComponent (Variables|GetFortPlayerControllerAthena)) "(TagName=\"Ability.Orb.RadiusMarking\")")
        (if _returnvalue
          (Utilities|Map|Remove (Variables|Default|GetAS_RadiusMarking_Map) (Variables|GetFortPlayerControllerAthena))
          (Ability|UnequipFortAbilitySet _value)
          (Fortnite|FortShippingLog 0 (Utilities|String|Append "BGA_DCN_OrbActor: Unequipping ability for player " (Utilities|GetDisplayName (Variables|GetFortPlayerControllerAthena))))
          (else
            (Fortnite|FortShippingLog 0 "BGA_DCN_OrbActor: Trying to remove the ability set, but it doesn't exist on the map.")))))
    (:then_1
      (bind _returnvalue_2 (Ability|EquipFortAbilitySet (Variables|GetFortPlayerControllerAthena) "/DragonCartNarrative/Gameplay/Abilities/AS_DCN_OrbActor_RadiusMarking.AS_DCN_OrbActor_RadiusMarking"))
      (Utilities|Map|Add (Variables|Default|GetAS_RadiusMarking_Map) (Variables|GetFortPlayerControllerAthena) _returnvalue_2)
      (Development|PrintString (Utilities|String|Append "BGA_DCN_OrbActor: " (Utilities|GetDisplayName (Variables|GetFortPlayerControllerAthena)) " HAS SCAN ABILITY") true true "(R=0.000000,G=0.660000,B=1.000000,A=1.000000)" 1e+01)
      (return))))

(fn KnockbackVehicle
  (params (Vehicle :type Fort%20Athena%20Vehicle%20Object%20Reference))
  (Networking|SwitchHasAuthority
    (:Authority
      (Vehicle|WakeUp Vehicle)
      (Components|Movement|StopMovementImmediately (Pawn|GetMovementComponent Vehicle))
      (GameplayEffects|ApplyGameplayEffectToTarget :self (Vehicle|GetVehicleAbilitySystemComponent Vehicle) :GameplayEffectClass (LocalVariables|GetVehicleDisableEffect) :Target (Vehicle|GetVehicleAbilitySystemComponent Vehicle) :Level 0.0)
      (if (Class|FortVehicleInterface|IsBoosting Vehicle)
        (Physics|AddRadialForce (Variables|/Script/FortniteGame.FortAthenaVehicle|GetSkeletalMesh Vehicle) (Transformation|GetWorldLocation (Variables|Default|GetShieldImpulse)) 8e+02 (ScalableFloat|GetValueatLevelwithContext (Variables|Default|GetVehicle%20Knockback%20Radial%20Force)) "RIF_Constant" true) ; Apply Force to cancel out boost.
      )
      (Audio|BroadcastAbilityImpactatLocation 0 Vehicle (CallFunction|GetVehicleImpulseValue Vehicle) (ScalableFloat|GetValueatLevelwithContext (Variables|Default|GetVehicle%20Knockback%20Impulse%20Radius))))))

(fn LogIt
  (params (InString :type String) (bLogAsWarning :type Boolean))
  (Fortnite|FortShippingLog self (Utilities|String|Append "[BGA_DCN_OrbActor] - " InString) bLogAsWarning))

(fn LookAtAndShootTarget
  (params (TargetActor :type Actor%20Object%20Reference))
  (Utilities|IsValid TargetActor
    (:Is%20Valid
      (Class|FortActorFocusComponent|SetActorFocus (Variables|Default|GetFortActorFocus) TargetActor)
      (CallFunction|StartCheckAngle TargetActor))))

(fn OnRep_FinalHit)

(fn OnRep_HitCount
  (bind _shieldorbvisual (Variables|Default|GetShieldOrbVisual))
  ;[0.000877 0.1122 0.15 0.5] Set Hit Time in Shield material
  (Networking|CanExecuteCosmeticEvents
    (:True
      (Utilities|IsValid _shieldorbvisual
        (:Is%20Valid
          (Rendering|Material|SetScalarParameterforCustomPrimitiveData (Variables|/Script/FortniteGame.BuildingSMActor|GetStaticMeshComponent _shieldorbvisual) "Hit Time" (Utilities|Time|GetGameTimeinSeconds)))))))

(fn OnRep_InShootingMode
  (CallFunction|SetOrbModeVisuals (Variables|ShootingMode|GetInShootingMode))
  (CallFunction|UpdateOrbRotationAudio)
  (CallFunction|ShootingModeStartedAudio))

(fn OnRep_OrbisAggro
  (bind _orb_is_aggro (Variables|DarkVoyagerEcho|GetOrb%20is%20Aggro))
  (CallFunction|LogIt (Utilities|String|Append "Orb changed aggro state to: " (Utilities|String|SelectString "AGGRO" "not aggro" (Variables|DarkVoyagerEcho|GetOrb%20is%20Aggro)) " (" (Utilities|String|SelectString "server" "client" (Networking|IsServer)) ")"))
  (Components|Activation|SetActive (Variables|Default|GetNiagara%20-%20aggro%20state) _orb_is_aggro false)
  (Rendering|Material|SetScalarParameterValue "/DragonCartNarrative/Assets/VFX/RiftPortalIdle/MPC_DragonCart_Orb_Rift.MPC_DragonCart_Orb_Rift" "Orb Hostile Mode" (Math|Conversions|ToFloat(Boolean) _orb_is_aggro) :!class /Script/Engine.KismetMaterialLibrary)
  (if (Networking|IsServer)
    (CallFunction|OrbChangedAggroState(Server))
    (else
      (CallFunction|OrbChangedAggroState(Client)))))

(fn OnRep_OverlapResultStruct
  (Networking|SwitchHasAuthority
    (:Remote
      (CallFunction|PlayKnockbackGCN (Variables|Default|GetOverlapResultStruct)))))

(fn OnRep_ShowOrb
  (Networking|CanExecuteCosmeticEvents
    (:True
      (Rendering|SetVisibility (Variables|Default|GetStaticMesh) (Variables|Default|GetShowOrb) true))))

(fn OrbLaserCharge
  (params (Target :type Actor%20Object%20Reference))
  (Ability|SendGameplayEventtoActor self "(TagName=\"Event.DragonCart.Orb.BeamBlast.Fire\")" (Utilities|Struct|MakeGameplayEventData "(TagName=\"\")" self Target)))

(fn PingPlayer
  (params (PlayerController :type Fort%20Player%20Controller%20Athena%20Object%20Reference) (PingOrigin :type Vector) (PingRadius :type Float%20%28double-precision%29) (PawnsToBeMarked :type Array%20of%20Actor%20Object%20References) (ChestsToBeMarked :type Array%20of%20Actor%20Object%20References) (ExtraDelayMultiplier :type Float%20%28double-precision%29))
  (bind _output_get (LocalVariables|SetPlayerPawn (Game|Player|GetPlayerPawn PlayerController)))
  (Utilities|IsValid _output_get
    (:Is%20Valid
      (CallFunction|LogIt (Utilities|String|Append "Pinging player " (Utilities|GetDisplayName (Variables|GetPlayerController)) " (" (Utilities|String|ToString(Integer) (Utilities|Array|Length (Variables|GetPawnsToBeMarked))) " pawns and " (Utilities|String|ToString(Integer) (Utilities|Array|Length (Variables|GetChestsToBeMarked))) " chests to be marked)"))
      ; If the player is within the inner ping radius
      (if (<= (Math|Vector|Distance(Vector) (Transformation|GetActorLocation (LocalVariables|GetPlayerPawn)) (Variables|GetPingOrigin)) (Variables|GetPingRadius))
        ; Equip and activate the RadiusMarking ability on the player
        (CallFunction|HandlePlayerAbilitySet (Variables|GetPlayerController)) ; Equip the RadiusMarking ability set
        (bind _returnvalue (Ability|TargetData|AppendTargetDataHandle (Ability|TargetData|AbilityTargetDatafromActorArray (Variables|GetPawnsToBeMarked)) (Ability|TargetData|AbilityTargetDatafromActorArray (Variables|GetChestsToBeMarked))))
        (Ability|SendGameplayEventtoActor (LocalVariables|GetPlayerPawn) "(TagName=\"Event.DragonCart.Orb.Scan\")" (Utilities|Struct|MakeGameplayEventData "(TagName=\"\")" self 0 0 0 "()" "(GameplayTags=)" "(GameplayTags=)" (Variables|GetExtraDelayMultiplier) _returnvalue)) ; Send a gameplay event to activate the ability we just equipped
      ))) ; Player Pawn Valid?
)

(fn RevealOrb
  (CallFunction|LogIt "Revealing Orb")
  (Ability|ExecuteGameplayCueWithParamsReplicated (Variables|ShootingMode|GetTargeting%20Pulse%20Visual%20GC) (Ability|GameplayCue|MakeGameplayCueParameters :NormalizedMagnitude 0.0 :RawMagnitude (ScalableFloat|GetValueatLevel (Variables|Default|GetHF_Ping_VisualRadius)) :Location (Transformation|GetSocketLocation (Variables|Default|GetStaticMesh) "EyeSocket") :Normal "0,0,0" :GameplayEffectLevel 1 :AbilityLevel 1 :bReplicateLocationWhenUsingMinimalRepProxy false)))

(fn SelectValidTarget
  ; Ensure we ignore the orb when targeting, ping all players, then pick a visible player as a shot target.
  (Networking|SwitchHasAuthority
    (:Authority
      (Rendering|Debug|DrawDebugSphere (Transformation|GetActorLocation self) (ScalableFloat|GetValueatLevel (Variables|Default|GetVisualPingRadius)) 12 "(R=1.000000,G=0.798989,B=0.000000,A=1.000000)" 3e+01 1e+01)
      (bind _pawnswithinnormalradius (CallFunction|FindPawnsWithinPingDistance))
      (Utilities|Array|Clear (Variables|Default|GetValidTargets)) ; Clear the list of \nplayer targets.
      (for _array_element _pawnswithinnormalradius
        (if (< (Transformation|GetHorizontalDistanceTo _array_element self) (ScalableFloat|GetValueatLevelwithContext (Variables|ShootingMode|GetOrbShootingRange)))
          (Development|PrintString "Target in range")
          (bind _outhit (Collision|LineTraceByChannel (Transformation|GetSocketLocation (Variables|Default|GetStaticMesh) "EyeSocket") (Transformation|GetActorLocation _array_element) "TraceTypeQuery1" true (Utilities|Array|MakeArray self) "ForDuration" true "(R=1.000000,G=0.000000,B=0.000000,A=1.000000)" "(R=0.000000,G=1.000000,B=0.000000,A=1.000000)" 5e+01))
          (if (== (.HitActor (Collision|BreakHitResult _outhit)) _array_element)
            (Utilities|Array|Add (Variables|Default|GetValidTargets) _array_element))))
      (Utilities|IsValid (Variables|Default|GetInstigatingPlayer)
        (:Is%20Valid
          (CallFunction|LookatandShootTarget (Variables|Default|GetInstigatingPlayer)) ; shoot at the player who shot the Orb
        )
        (:Is%20Not%20Valid
          (bind _element (ArrayMacros|GetRandomArrayElement (Variables|Default|GetValidTargets))
            (:%20%20
              (CallFunction|LookatandShootTarget _element) ; shoot at a random player in the area
            )))) ; Is there a player who shot the Orb?\nIf so, shoot at them. If not, pull \na random target from the nearby \nplayer list.
    )))

(fn TriggerNDCImpact
  (params (Hit :type Hit%20Result%20Structure%20%28by%20ref%29) (Damage :type Float%20%28double-precision%29))
  (bind (_bblockinghit _binitialoverlap _time _distance _location _impactpoint _normal _impactnormal _physmat _hitactor _hitcomponent _hitbonename _bonename _hititem _elementindex _faceindex _tracestart _traceend) (Collision|BreakHitResult Hit))
  (bind _returnvalue (Math|Float|NormalizetoRange Damage 1.0 1.8e+02))
  (CallFunction|TriggerRipple _location (Math|Float|Lerp 5e+02 5e+03 _returnvalue) (Math|Float|Lerp 0.3 1.0 _returnvalue))
  (NiagaraDataChannel|WriteToNiagaraDataChannel(Legacy) :Channel "/DragonCartNarrative/Assets/VFX/OrbBarrier/OrbBarrierImpact/NDC_DCN_OrbImpact.NDC_DCN_OrbImpact" :SearchParams (Utilities|Struct|MakeNiagaraDataChannelSearchParameters 0 _location) :bVisibleToBlueprint true :bVisibleToNiagaraCPU true :bVisibleToNiagaraGPU true :ImpactPoint _location :ImpactNormal _normal :ImpactPhysicalSurface "SurfaceType_Default" :ScaleMultiplier _returnvalue :ImpactReflectionVector (Math|Vector|GetReflectionVector (Math|Vector|GetUnitDirection(Vector) _tracestart _traceend) _normal)))

(fn TriggerRipple
  (params (Position :type Vector) (Radius :type Float%20%28double-precision%29) (Duration :type Float%20%28double-precision%29))
  (bind _ripplecounter (Variables|Default|GetRippleCounter))
  (bind _returnvalue (Utilities|String|ToString(Integer) _ripplecounter))
  (bind _shieldmesh (Variables|Default|GetShieldMesh))
  (bind _ripplecounter_2 (Variables|Default|GetRippleCounter))
  (Rendering|Material|SetVectorParameterValueonMaterials _shieldmesh (Utilities|String|StringToName (Utilities|String|Append "RipplePosition" _returnvalue)) Position)
  (Rendering|Material|SetScalarParameterValueonMaterials _shieldmesh (Utilities|String|StringToName (Utilities|String|Append "RippleRadius" _returnvalue)) Radius)
  (Rendering|Material|SetScalarParameterValueonMaterials _shieldmesh (Utilities|String|StringToName (Utilities|String|Append "RippleDuration" _returnvalue)) Duration)
  (Rendering|Material|SetScalarParameterValueonMaterials _shieldmesh (Utilities|String|StringToName (Utilities|String|Append "RippleTime" _returnvalue)) (Utilities|Time|GetGameTimeinSeconds))
  (if (>= _ripplecounter_2 2)
    (Variables|Default|SetRippleCounter 0)
    (else
      (Math|Integer|IncrementInt _ripplecounter_2))))

(fn UpdateLaserEnabledState
  ; Orb Laser - check hotfix and calendar values
  (if (ScalableFloat|GetValueasBool (Variables|Default|GetHF_DeathBeam_Enabled))
    (if (ScalableFloat|GetValueasBool (Variables|Default|GetHF_DeathBeam_ControlFromCalendar))
      (bind _biseventactive (CalendarEvents|GetCalendarEventInformation (Variables|Default|GetCalendar%20-%20Laser%20Controller))) ; laser calendar event active?
      (if _biseventactive
        (Variables|Default|SetShootingEnabled true) ; Laser enabled from hotfix.\nCalendar controlled & Active!\n..enable shooting
        (else
          (Variables|Default|SetPingsEnabled false) ; Laser enabled from hotfix.\nCalendar controlled & NOT active.\n..disable shooting
        ))
      (else
        (Variables|Default|SetShootingEnabled true) ; Shooting enabled from hotfix.\nNOT calendar controlled.\n..enable shooting.
      ))
    (else
      (Variables|Default|SetShootingEnabled false) ; Shooting Disabled from hotfix.
    ))
  (CallFunction|LogIt (Utilities|String|Append "Orb Laser enabled: " (Utilities|String|ToString(Boolean) (Variables|Default|GetShootingEnabled)) ", calendar controlled: " (Utilities|String|ToString(Boolean) (ScalableFloat|GetValueasBool (Variables|Default|GetHF_DeathBeam_ControlFromCalendar)))))
  (return))

(fn UpdatePingScanEnabledState
  (bind _hf_deathbeam_scanwhenshot_enabled (Variables|Default|GetHF_DeathBeam_ScanWhenShot_Enabled))
  (bind _returnvalue (ScalableFloat|GetValueasBool _hf_deathbeam_scanwhenshot_enabled))
  ; Orb Scan/Ping - check hotfixes and calendar events
  (if (ScalableFloat|GetValueasBool (Variables|Default|GetHF_ScanEnabled))
    (if (ScalableFloat|GetValueasBool (Variables|Default|GetHF_Scan_ControlFromCalendar))
      (bind _biseventactive (CalendarEvents|GetCalendarEventInformation (Variables|Default|GetCalendar%20-%20Scan%20Controller))) ; scan/ping calendar event active?
      (if _biseventactive
        (Variables|Default|SetPingsEnabled true) ; Scan enabled.from hotfix.\nCalendar controlled.\nEnable the ping.
        (Variables|Default|SetPingsEnabled_WhenShootingLaser _returnvalue)
        (else
          (Variables|Default|SetPingsEnabled false) ; Scan enabled.from hotfix.\nCalendar controlled.\nDisable the ping.
          (Variables|Default|SetPingsEnabled_WhenShootingLaser false)))
      (else
        (Variables|Default|SetPingsEnabled true) ; Scan enabled.from hotfix.\nNOT Calendar controlled.\nEnable the ping.
        (Variables|Default|SetPingsEnabled_WhenShootingLaser _returnvalue)))
    (else
      (Variables|Default|SetPingsEnabled false)
      (Variables|Default|SetPingsEnabled_WhenShootingLaser false)))
  (if (Variables|Default|GetScan%20-%20Force%20Enable%20fromCheat)
    (Variables|Default|SetScan%20-%20Force%20Enable%20fromCheat true))
  (CallFunction|LogIt (Utilities|String|Append "Orb Ping/Scan enabled: " (Utilities|String|ToString(Boolean) (Variables|Default|GetPingsEnabled)) ", calendar control: " (Utilities|String|ToString(Boolean) (ScalableFloat|GetValueasBool (Variables|Default|GetHF_Scan_ControlFromCalendar))) ", Ping when shooting laser: " (Utilities|String|ToString(Boolean) (Variables|Default|GetPingsEnabled_WhenShootingLaser)) ", Force enable from cheat: " (Utilities|String|ToString(Boolean) (Variables|Default|GetScan%20-%20Force%20Enable%20fromCheat))))
  (return))

(fn ConstructionScript)
