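; Voyager shield gameplay actor for the flying orb that repels whatever enters it and shows impact effects.
; On server overlap it knocks back non-NPC players with fall immunity and trace-based collision damage, tosses pickups, and disables and launches vehicles.
; A replicated overlap result drives the client knockback cue.
; Hits feed a Niagara Data Channel and damage-scaled ripples, cycling three indexed material parameter slots.
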(event EventDestroyed
  (|CallParentFunction))

(event EventBeginPlay
  (bind _bga_health (Variables|Default|GetBGA%20Health))
  ;[0 1 0.06226] Set Orb as Invulnerable at start
  (Class|FortAsyncActionCurrentPlaylistReady|PlaylistDataReadyAsync
    (:OnReady
      (Health|SetMaxHealthDontCreateASC _bga_health)
      (Health|SetHealth _bga_health))))

(event OnComponentBeginOverlap(ShieldImpulse)
  (params (OverlappedComponent :type Primitive%20Component%20Object%20Reference) (OtherActor :type Actor%20Object%20Reference) (OtherComp :type Primitive%20Component%20Object%20Reference) (OtherBodyIndex :type Integer) (bFromSweep :type Boolean) (SweepResult :type Hit%20Result%20Structure%20%28by%20ref%29))
  (bind (_bblockinghit _binitialoverlap _time _distance _location) (Collision|BreakHitResult SweepResult))
  (bind _self self)
  ;[0.000304 0.04519 0.06125] Knockback players entering the shield.
  (Networking|SwitchHasAuthority
    (:Authority
      (bind _asfort_player_pawn_athena (Utilities|Casting|CastToFortPlayerPawnAthena OtherActor)
        (:then
          (if (not (GameplayTags|HasMatchingGameplayTag _asfort_player_pawn_athena "(TagName=\"Pawn.Athena.NPC\")"))
            (Variables|Default|SetOverlapResultStruct SweepResult)
            (bind _outhits (Collision|MultiCapsuleTraceByChannel :Start (Transformation|GetActorLocation OtherActor) :End (+ (Transformation|GetActorLocation OtherActor) (* (Math|Vector|GetUnitDirection(Vector) (Transformation|GetActorLocation OtherActor) _location) (ScalableFloat|GetValueatLevel (Variables|Default|GetDestroyDistance)))) :Radius 8e+01 :HalfHeight 1.5e+02 :TraceChannel "TraceTypeQuery11" :bTraceComplex false :DrawDebugType "ForDuration" :bIgnoreSelf true :TraceColor "(R=1.000000,G=0.000000,B=0.000000,A=1.000000)" :TraceHitColor "(R=0.000000,G=1.000000,B=0.000000,A=1.000000)" :DrawTime 2.0))
            (for _array_element _outhits
              (if (not (Math|Boolean|ANDBoolean (Variables|/Script/Engine.Actor|GetbCanBeDamaged (.HitActor (Collision|BreakHitResult _array_element))) (Utilities|IsA (.HitActor (Collision|BreakHitResult _array_element)) "/Script/FortniteGame.BuildingActor") (Utilities|IsA (.HitActor (Collision|BreakHitResult _array_element)) "/Script/FortniteGame.BuildingTrap")))
                (GameplayEffects|ApplyGameplayEffectToTarget :self (Ability|GetAbilitySystemComponent _asfort_player_pawn_athena) :GameplayEffectClass "/DragonCartNarrative/Gameplay/FlyingOrb/Abilities/GE_DCN_Orb_Knockback_Damage.GE_DCN_Orb_Knockback_Damage_C" :Target (Ability|GetAbilitySystemComponent (.HitActor (Collision|BreakHitResult _array_element))) :Level 0.0)))
            (Utilities|FlowControl|DelayUntilNextTick)
            (GameplayEffects|FortApplyGameplayEffectToSelf :self (Ability|GetFortAbilitySystemComponent _asfort_player_pawn_athena) :GameplayEffectClass "/DragonCartNarrative/Gameplay/FlyingOrb/Abilities/GE_Voyager_Shield_Knockback.GE_Voyager_Shield_Knockback_C" :OptionalSourceObject _self :OptionalEffectCauser _self :Level 0.0)
            (GameplayEffects|FortApplyGameplayEffectToSelf :self (Ability|GetFortAbilitySystemComponent _asfort_player_pawn_athena) :GameplayEffectClass "/Game/Athena/Items/Parents/FallImmunity/GE_Athena_TillLandFallDamageImmunity_NoFX_Grant_ApplyInAirOnly.GE_Athena_TillLandFallDamageImmunity_NoFX_Grant_ApplyInAirOnly_C" :OptionalSourceObject _self :OptionalEffectCauser _self :Level 0.0)))
        (:CastFailed
          (bind _asfort_pickup (Utilities|Casting|CastToFortPickup OtherActor)
            (:then
              (bind (_returnvalue_roll _returnvalue_pitch _returnvalue_yaw) (Math|Rotator|BreakRotator (Math|Rotator|FindLookatRotation (Transformation|GetActorLocation) (Transformation|GetActorLocation _asfort_pickup))))
              (Pickup|TossPickupinDirection _asfort_pickup (* (Math|Vector|MakeVector _returnvalue_roll _returnvalue_pitch (Math|Float|Absolute(Float) _returnvalue_yaw)) 5.0) 0 false true))
            (:CastFailed
              (bind _asfort_athena_vehicle (Utilities|Casting|CastToFortAthenaVehicle OtherActor))
              (CallFunction|KnockbackVehicle _asfort_athena_vehicle))))))))

(event Custom|PlayKnockbackGCN
  (params (OverlapResultStruct :type Hit%20Result%20Structure))
  (bind (_bblockinghit _binitialoverlap _time _distance _location _impactpoint _normal _impactnormal _physmat _hitactor) (Collision|BreakHitResult OverlapResultStruct))
  (Utilities|FlowControl|DoOnce false
    (:Completed
      (Utilities|FlowControl|Sequence
        (:then_0
          (Utilities|FlowControl|Delay 0.25))
        (:then_1
          _))
      (CallFunction|TriggerRipple _location 3.0 2.0)
      (GameplayCue|ExecuteGameplayCue%28Burst%29Local _hitactor (Variables|Default|GetKnockbackGCTag) (Ability|GameplayCue|MakeGameplayCueParameters :NormalizedMagnitude 0.0 :RawMagnitude 0.0 :Location _location :Normal (* _normal -1.0) :GameplayEffectLevel 1 :AbilityLevel 1 :bReplicateLocationWhenUsingMinimalRepProxy false)))))

(event EventOnDamagePlayEffects
  (params (Damage :type Float%20%28single-precision%29) (DamageTags :type Gameplay%20Tag%20Container%20Structure%20%28by%20ref%29) (Momentum :type Vector :default "0, 0, 0") (HitInfo :type Hit%20Result%20Structure%20%28by%20ref%29) (InstigatedBy :type Fort%20Pawn%20Object%20Reference) (DamageCauser :type Actor%20Object%20Reference) (EffectContext :type Gameplay%20Effect%20Context%20Handle%20Structure))
  ;[0.1194 0.1035 0.15 0.5] Trigger NDC impact on weapon hits
  (CallFunction|TriggerNDCImpact HitInfo Damage)
  (Development|PrintString "Triggered NDC On Weapon Hit")
  (Development|PrintString (Utilities|String|Append "I've been shot by: " (Utilities|GetDisplayName InstigatedBy) " and they are at: " (Utilities|String|ToString(Vector) (Transformation|GetActorLocation InstigatedBy)) " and they dealt a damage total of: " (Utilities|String|ToString(Float) Damage)) true true "(R=0.000000,G=0.660000,B=1.000000,A=1.000000)" 2e+01))

(fn GetMarkerDisplayTeam
  (bind _as_fort_player_controller_athena (Variables|Default|GetAs%20Fort%20Player%20Controller%20Athena))
  (bind _returnvalue (Game|GetInstigator _as_fort_player_controller_athena))
  (Utilities|IsValid _as_fort_player_controller_athena
    (:Is%20Valid
      (Utilities|IsValid _returnvalue
        (:Is%20Valid
          (return (Team|GetActorTeam _returnvalue)))
        (:Is%20Not%20Valid
          _)))
    (:Is%20Not%20Valid
      _))
  (return 255))

(fn GetVehicleImpulseValue
  (params (Actor :type Actor%20Object%20Reference))
  (bind (_returnvalue_x _returnvalue_y) (Math|Vector|BreakVector (- (Transformation|GetActorLocation Actor) (Transformation|GetWorldLocation (Variables|Default|GetShieldMesh)))))
  (return (+ (* (Math|Vector|Normalize (Math|Vector|MakeVector _returnvalue_x _returnvalue_y)) (Math|Vector|MakeVector (ScalableFloat|GetValueatLevelwithContext (Variables|Default|GetVehicle%20Knockback%20-%20Horz)) (ScalableFloat|GetValueatLevelwithContext (Variables|Default|GetVehicle%20Knockback%20-%20Horz)))) (Math|Vector|MakeVector 0.0 0.0 (ScalableFloat|GetValueatLevelwithContext (Variables|Default|GetVehicle%20Knockback%20-%20Vert))))))

(fn KnockbackVehicle
  (params (Vehicle :type Fort%20Athena%20Vehicle%20Object%20Reference))
  (Networking|SwitchHasAuthority
    (:Authority
      (Vehicle|WakeUp Vehicle)
      (Components|Movement|StopMovementImmediately (Pawn|GetMovementComponent Vehicle))
      (GameplayEffects|ApplyGameplayEffectToTarget :self (Vehicle|GetVehicleAbilitySystemComponent Vehicle) :GameplayEffectClass (LocalVariables|GetVehicleDisableEffect) :Target (Vehicle|GetVehicleAbilitySystemComponent Vehicle) :Level 0.0)
      (if (Class|FortVehicleInterface|IsBoosting Vehicle)
        (Physics|AddRadialForce (Variables|/Script/FortniteGame.FortAthenaVehicle|GetSkeletalMesh Vehicle) (Transformation|GetWorldLocation (Variables|Default|GetShieldMesh)) 8e+02 (ScalableFloat|GetValueatLevelwithContext (Variables|Default|GetVehicle%20Knockback%20Radial%20Force)) "RIF_Constant" true) ; Apply Force to cancel out boost.
      )
      (Audio|BroadcastAbilityImpactatLocation 0 Vehicle (CallFunction|GetVehicleImpulseValue Vehicle) (ScalableFloat|GetValueatLevelwithContext (Variables|Default|GetVehicle%20Knockback%20Impulse%20Radius))))))

(fn OnRep_HitCount
  (bind _shieldorbvisual (Variables|Default|GetShieldOrbVisual))
  ;[0.000877 0.1122 0.15 0.5] Set Hit Time in Shield material
  (Networking|CanExecuteCosmeticEvents
    (:True
      (Utilities|IsValid _shieldorbvisual
        (:Is%20Valid
          (Rendering|Material|SetScalarParameterforCustomPrimitiveData (Variables|/Script/FortniteGame.BuildingSMActor|GetStaticMeshComponent _shieldorbvisual) "Hit Time" (Utilities|Time|GetGameTimeinSeconds)))))))

(fn OnRep_OverlapResultStruct
  (Networking|SwitchHasAuthority
    (:Remote
      (CallFunction|PlayKnockbackGCN (Variables|Default|GetOverlapResultStruct)))))

(fn TriggerNDCImpact
  (params (Hit :type Hit%20Result%20Structure%20%28by%20ref%29) (Damage :type Float%20%28double-precision%29))
  (bind (_bblockinghit _binitialoverlap _time _distance _location _impactpoint _normal _impactnormal _physmat _hitactor _hitcomponent _hitbonename _bonename _hititem _elementindex _faceindex _tracestart _traceend) (Collision|BreakHitResult Hit))
  (bind _returnvalue (Math|Float|NormalizetoRange Damage 1.0 1.8e+02))
  (CallFunction|TriggerRipple _location (Math|Float|Lerp 6e+02 1.5e+03 _returnvalue) (Math|Float|Lerp 0.4 1.0 _returnvalue))
  (NiagaraDataChannel|WriteToNiagaraDataChannel(Legacy) :Channel "/DragonCartNarrative/Assets/VFX/OrbBarrier/OrbBarrierImpact/NDC_DCN_OrbImpact.NDC_DCN_OrbImpact" :SearchParams (Utilities|Struct|MakeNiagaraDataChannelSearchParameters 0 _location) :bVisibleToBlueprint true :bVisibleToNiagaraCPU true :bVisibleToNiagaraGPU true :ImpactPoint _location :ImpactNormal _normal :ImpactPhysicalSurface "SurfaceType_Default" :ScaleMultiplier (* _returnvalue 0.1) :ImpactReflectionVector (Math|Vector|GetReflectionVector (Math|Vector|GetUnitDirection(Vector) _tracestart _traceend) _normal)))

(fn TriggerRipple
  (params (Position :type Vector) (Radius :type Float%20%28double-precision%29) (Duration :type Float%20%28double-precision%29))
  (bind _ripplecounter (Variables|Default|GetRippleCounter))
  (bind _returnvalue (Utilities|String|ToString(Integer) _ripplecounter))
  (bind _shieldmesh (Variables|Default|GetShieldMesh))
  (bind _ripplecounter_2 (Variables|Default|GetRippleCounter))
  (Rendering|Material|SetVectorParameterValueonMaterials _shieldmesh (Utilities|String|StringToName (Utilities|String|Append "RipplePosition" _returnvalue)) Position)
  (Rendering|Material|SetScalarParameterValueonMaterials _shieldmesh (Utilities|String|StringToName (Utilities|String|Append "RippleRadius" _returnvalue)) Radius)
  (Rendering|Material|SetScalarParameterValueonMaterials _shieldmesh (Utilities|String|StringToName (Utilities|String|Append "RippleDuration" _returnvalue)) Duration)
  (Rendering|Material|SetScalarParameterValueonMaterials _shieldmesh (Utilities|String|StringToName (Utilities|String|Append "RippleTime" _returnvalue)) (Utilities|Time|GetGameTimeinSeconds))
  (if (>= _ripplecounter_2 2)
    (Variables|Default|SetRippleCounter 0)
    (else
      (Math|Integer|IncrementInt _ripplecounter_2))))

(fn ConstructionScript)
