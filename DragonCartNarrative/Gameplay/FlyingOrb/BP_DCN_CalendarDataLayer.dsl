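; Server-driven Blueprint that stages the DragonCart narrative's orb, rift and shard deliveries by switching World Partition data layers from calendar events.
; Once playlist data and calendar sync are ready, watched events call OrbTechHQPhase, which activates and unloads layers from a per-phase table; a hotfix can substitute a fallback phase.
; RepNotify rift scale drives a material parameter collection.
; Custom events replay stages directly or pick random ocean orb layers.
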(event Custom|SetOrbVisibility)

(event EventBeginPlay
  (bind _rotatingorbcalendars (Variables|Default|GetRotatingOrbCalendars))
  ;[0.6823 0.3546 0.6748 0.5] Begin Play / Calendar / Choose Orb location
  (if (Networking|IsServer)
    (CallFunction|LogIt "Begin Play - server" false)
    (Class|FortAsyncActionCurrentPlaylistReady|PlaylistDataReadyAsync
      (:OnReady
        (if (ScalableFloat|GetValueasBool (Variables|Hotfixes|GetBypassCalendarData))
          ;; Phase to set if not using calendar
          (CallFunction|OrbTechHQPhase (Math|Float|Truncate (ScalableFloat|GetValueatLevel (Variables|Hotfixes|GetFallBackPhase))))
          (else
            (MCPContext|InitialCalendarSyncCompleted
              (:OnSyncComplete
                (Utilities|FlowControl|Sequence
                  (:then_0
                    (MCPContext|WatchCalendarEvent (Variables|Default|GetOrbRiftingIn)
                      (:OnEventActive
                        (CallFunction|OrbTechHQPhase 10) ; Orb Rift In
                      )))
                  (:then_1
                    (MCPContext|WatchCalendarEvent (Variables|Default|GetOrb%20Rifting%20out)
                      (:OnEventActive
                        (CallFunction|LogIt "Orb rifting out" false)
                        (CallFunction|OrbTechHQPhase (Variables|Default|GetRiftoutPhase)) ; rift out
                      )))
                  (:then_2
                    (bind _eventname (MCPContext|WatchCalendarEvents _rotatingorbcalendars)
                      (:OnEventActive
                        (CallFunction|LogIt "Orb rotating is active" false)
                        (CallFunction|OrbTechHQPhase (+ (+ (Utilities|Array|FindItem _rotatingorbcalendars _eventname) 1) (Variables|Default|GetRiftoutPhase))))))
                  (:then_3
                    (MCPContext|WatchCalendarEvent (Variables|Default|GetCalendar%20-%20Shock)
                      (:OnEventActive
                        (CallFunction|LogIt "Shock Calendar Event Active" false)
                        (CallFunction|OrbTechHQPhase 18) ; Shock
                      )))
                  (:then_4
                    (bind _eventname_2 (MCPContext|WatchCalendarEvents (Variables|Default|GetRiftGrowCalendars)))
                    (switch string _eventname_2
                      (:DC_RiftGrow1
                        (Variables|Default|SetSetRiftScale 0.3))
                      (:DC_RiftGrow2
                        (Variables|Default|SetSetRiftScale 0.7))
                      (:DC_RiftGrowNoOrb1
                        (Variables|Default|SetSetRiftScale 0.3)
                        (CallFunction|OrbTechHQPhase 16))
                      (:DC_RiftGrowNoOrb2
                        (Variables|Default|SetSetRiftScale 0.7)))
                    (CallFunction|OrbTechHQPhase 17))
                  (:then_5
                    (MCPContext|WatchCalendarEvent "DC_OrbTechHQ"
                      (:OnEventActive
                        (CallFunction|OrbTechHQPhase 11))))
                  (:then_6
                    (bind _eventname_3 (MCPContext|WatchCalendarEvents (Variables|Default|GetDeliveries)))
                    (switch string _eventname_3
                      (:DC_ShardDelivery1
                        (CallFunction|OrbTechHQPhase 12))
                      (:DC_ShardDelivery2
                        (CallFunction|OrbTechHQPhase 13))
                      (:DC_ShardDelivery3
                        (CallFunction|OrbTechHQPhase 14) ; 2 shard pieces
                      )))
                  (:then_7
                    (MCPContext|WatchCalendarEvent "DC_OrbTechHQVoyager"
                      (:OnEventActive
                        (CallFunction|OrbTechHQPhase 15))))
                  (:then_8
                    (MCPContext|WatchCalendarEvent "RL_Rift2"
                      (:OnEventActive
                        (CallFunction|SetRedRiftLoaded true))))
                  (:then_9
                    (MCPContext|WatchCalendarEvent "DC_RealityRifts"
                      (:OnEventActive
                        (CallFunction|OrbTechHQPhase 19))))
                  (:then_10
                    (Utilities|FlowControl|Sequence
                      (:then_0
                        (bind _biseventactive (CalendarEvents|GetCalendarEventInformation "DC_NoOrb"))
                        (if _biseventactive
                          (CallFunction|OrbTechHQPhase 1) ; no orb
                        ))
                      (:then_1
                        (bind _biseventactive_2 (CalendarEvents|GetCalendarEventInformation "DC_3ShardsVisible"))
                        (if _biseventactive_2
                          (CallFunction|OrbTechHQPhase 21) ; 3 shard pieces
                        ))
                      (:then_2
                        (bind _biseventactive_3 (CalendarEvents|GetCalendarEventInformation "DC_OrbOverEcho"))
                        (if _biseventactive_3
                          (bind _returnvalue (WorldPartition|GetDataLayerManager))
                          (DataLayers|SetDataLayerRuntimeState _returnvalue "/DragonCartNarrative/DataLayers/DL_DCN_Orb_DV_Echo.DL_DCN_Orb_DV_Echo" "Activated" false :!class /Script/Engine.DataLayerManager) ; data layer: Orb used for Dark Voyager Echoes
                        ))
                      (:then_3
                        (bind _biseventactive_4 (CalendarEvents|GetCalendarEventInformation "DC_OrbClosedOverTechHQ"))
                        (if _biseventactive_4
                          (bind _returnvalue_2 (WorldPartition|GetDataLayerManager))
                          (DataLayers|SetDataLayerRuntimeState _returnvalue_2 "/DragonCartNarrative/DataLayers/DL_DCN_Orb_TechHQ_OrbClose.DL_DCN_Orb_TechHQ_OrbClose" "Activated" false :!class /Script/Engine.DataLayerManager) ; closed orb over Tech HQ
                        ))))
                  (:then_11
                    (MCPContext|WatchCalendarEvent "DC_PreShock"
                      (:OnEventActive
                        (CallFunction|PreShock))))
                  (:then_12
                    (MCPContext|WatchCalendarEvent "DC_RiftGrowNoOrb1"
                      (:OnEventActive
                        (CallFunction|RiftStageNoOrb1))))
                  (:then_13
                    (MCPContext|WatchCalendarEvent "DC_RiftGrowNoOrb2"
                      (:OnEventActive
                        (CallFunction|RiftStageNoOrb2))))
                  (:then_14
                    ;; Delivery 4
                    (Utilities|FlowControl|Sequence
                      (:then_0
                        (bind _biseventactive_5 (CalendarEvents|GetCalendarEventInformation "DC_ShardDelivery4")) ; Mid Delivery
                        (if _biseventactive_5
                          (CallFunction|OrbTechHQPhase 23) ; 3 shard pieces
                        ))
                      (:then_1
                        (bind _biseventactive_6 (CalendarEvents|GetCalendarEventInformation "DC_ShardDelivered4")) ; Post Delivery
                        (if _biseventactive_6
                          (CallFunction|OrbTechHQPhase 24) ; 3 shard pieces
                        ))))
                  (:then_15
                    ;; Shard Delivery 4+
                    (bind _eventname_4 (MCPContext|WatchCalendarEvents (Variables|Default|GetShardEvents))
                      (:OnEventActive
                        (bind _biseventactive_7 (CalendarEvents|GetCalendarEventInformation _eventname_4)) ; Shards Delivered
                        (if _biseventactive_7
                          (switch string _eventname_4
                            (:DC_4ShardsVisible
                              (CallFunction|OrbTechHQPhase 22) ; 4 shard pieces
                            )
                            (:DC_5ShardsVisible
                              (CallFunction|OrbTechHQPhase 25) ; 5 shard pieces
                            )
                            (:DC_6ShardsVisible
                              (CallFunction|OrbTechHQPhase 26) ; 6 shard pieces
                            )
                            (:DC_7ShardsVisible
                              (CallFunction|OrbTechHQPhase 27) ; 6 shard pieces
                            ))))))
                  (:then_16
                    ;; Oasis Events - 5th Shard Delivery
                    (bind _eventname_5 (MCPContext|WatchCalendarEvents (Variables|Default|GetOasisEvents))
                      (:OnEventActive
                        (bind _biseventactive_8 (CalendarEvents|GetCalendarEventInformation _eventname_5)) ; Shards Delivered
                        (if _biseventactive_8
                          (switch string _eventname_5
                            (:DC_Oasis_RiftGrow1
                              (Variables|Default|SetSetRiftScale 0.3))
                            (:DC_Oasis_RiftGrow2
                              (Variables|Default|SetSetRiftScale 0.7))
                            (:DC_Oasis_Delivery1
                              (CallFunction|OrbTechHQPhase 29))
                            (:DC_Oasis_Delivery2
                              (CallFunction|OrbTechHQPhase 30))
                            (:DC_Oasis_Delivery3
                              (CallFunction|OrbTechHQPhase 31)
                              (Variables|Default|SetSetRiftScale 0.3)
                              (CallFunction|OrbTechHQPhase 17))
                            (:DC_Oasis_Delivered
                              (CallFunction|OrbTechHQPhase 37) ; DC_5ShardsVisible should be on at the same time
                              (Variables|Default|SetSetRiftScale 0.7)))
                          (CallFunction|OrbTechHQPhase 28)))))
                  (:then_17
                    ;; Squid Events - 6th Shard Delivery
                    (bind _eventname_6 (MCPContext|WatchCalendarEvents (Variables|Default|GetSquidEvents))
                      (:OnEventActive
                        (bind _biseventactive_9 (CalendarEvents|GetCalendarEventInformation _eventname_6)) ; Shards Delivered
                        (if _biseventactive_9
                          (switch string _eventname_6
                            (:DC_Squid_Rift1
                              (Variables|Default|SetSetRiftScale 0.3))
                            (:DC_Squid_Rift2
                              (Variables|Default|SetSetRiftScale 0.7))
                            (:DC_Squid_Delivery1
                              (CallFunction|OrbTechHQPhase 33))
                            (:DC_Squid_Delivery2
                              (CallFunction|OrbTechHQPhase 34))
                            (:DC_Squid_Delivery3
                              (CallFunction|OrbTechHQPhase 35)
                              (Variables|Default|SetSetRiftScale 0.3)
                              (CallFunction|OrbTechHQPhase 17))
                            (:DC_Squid_Delivered
                              (CallFunction|OrbTechHQPhase 36) ; DC_6ShardsVisible should be on at the same time
                              (Variables|Default|SetSetRiftScale 0.7)))
                          (CallFunction|OrbTechHQPhase 39)))))
                  (:then_18
                    (MCPContext|WatchCalendarEvent "DC_Squid_ST"
                      (:OnEventActive
                        (CallFunction|OrbTechHQPhase 38)
                        (Development|PrintString "Post Squid!" true true "(R=0.000000,G=0.660000,B=1.000000,A=1.000000)" 3e+02)))))
                (Utilities|FlowControl|DelayUntilNextTick)))))))))

(event Custom|PreShock
  ;[0.6823 0.3546 0.6748 0.5] Begin Play / Calendar / Choose Orb location
  (CallFunction|OrbTechHQPhase 20) ; PreShock
  (Rendering|Material|SetScalarParameterValue "/DragonCartNarrative/Assets/VFX/RiftPortalIdle/MPC_DragonCart_Orb_Rift.MPC_DragonCart_Orb_Rift" "ZPJitter" 1.0 :!class /Script/Engine.KismetMaterialLibrary)
  (Rendering|Material|SetScalarParameterValue "/DragonCartNarrative/Assets/VFX/RiftPortalIdle/MPC_DragonCart_Orb_Rift.MPC_DragonCart_Orb_Rift" "ZPOffset" 1.5 :!class /Script/Engine.KismetMaterialLibrary))

(event Custom|StartOrbRiftin
  ;[0.6823 0.3546 0.6748 0.5] Begin Play / Calendar / Choose Orb location
  ;; Orb Rift In / Rift Out
  (CallFunction|OrbTechHQPhase 10) ; This makes the data layer that has the rift in \nsequence visible. The Rift In sequence is set to\nauto play, so as soon as the data layer is made\nvisible it should play.
  (CallFunction|SetOrbVisibility))

(event Custom|StartOrbRiftOut
  ;[0.6823 0.3546 0.6748 0.5] Begin Play / Calendar / Choose Orb location
  ;; Orb Rift In / Rift Out
  (CallFunction|OrbTechHQPhase (Variables|Default|GetRiftoutPhase)) ; Phase 0 - disable all the Tech HQ data layers\n..not really needed as data layers should be\ndisabled by default.
)

(event Custom|RiftStage1
  ;[0.6823 0.3546 0.6748 0.5] Begin Play / Calendar / Choose Orb location
  (Variables|Default|SetSetRiftScale 0.3)
  (CallFunction|OrbTechHQPhase 17))

(event Custom|StartOrbShock
  ;[0.6823 0.3546 0.6748 0.5] Begin Play / Calendar / Choose Orb location
  ;; Orb Rift In / Rift Out
  (CallFunction|OrbTechHQPhase 18) ; Shock
)

(event Custom|RiftStage2
  ;[0.6823 0.3546 0.6748 0.5] Begin Play / Calendar / Choose Orb location
  (Variables|Default|SetSetRiftScale 0.7)
  (CallFunction|OrbTechHQPhase 17))

(event Custom|DC_DebugNow
  ;[0.6823 0.3546 0.6748 0.5] Begin Play / Calendar / Choose Orb location
  ;; Time Debug
  (Utilities|FlowControl|Delay 1.0)
  (bind _returnvalue (FortDateTime|GetMcpRelativeUtcNow (Fortnite|GetFirstLocalFortPlayerController true)))
  (CallFunction|LogIt (Utilities|String|Append "time now: " (Utilities|String|ToString(Integer) (.Month (Math|DateTime|BreakDateTime _returnvalue))) "/" (Utilities|String|ToString(Integer) (.Day (Math|DateTime|BreakDateTime _returnvalue))) ", " (Utilities|String|ToString(Integer) (.Hour (Math|DateTime|BreakDateTime _returnvalue))) ":" (Utilities|String|ToString(Integer) (.Minute (Math|DateTime|BreakDateTime _returnvalue))) ":" (Utilities|String|ToString(Integer) (.Second (Math|DateTime|BreakDateTime _returnvalue)))) false))

(event Custom|RiftStageNoOrb1
  ;[0.6823 0.3546 0.6748 0.5] Begin Play / Calendar / Choose Orb location
  (Variables|Default|SetSetRiftScale 0.3)
  (CallFunction|OrbTechHQPhase 16))

(event Custom|RiftStageNoOrb2
  ;[0.6823 0.3546 0.6748 0.5] Begin Play / Calendar / Choose Orb location
  (Variables|Default|SetSetRiftScale 0.7)
  (CallFunction|OrbTechHQPhase 16))

(event Custom|SetRedRiftLoaded
  (params (Loaded :type Boolean))
  ;[0.6823 0.3546 0.6748 0.5] Begin Play / Calendar / Choose Orb location
  (Networking|SwitchHasAuthority
    (:Authority
      (bind _returnvalue (WorldPartition|GetDataLayerManager))
      (DataLayers|SetDataLayerRuntimeState _returnvalue "/DragonCartNarrative/DataLayers/DL_RedRift_Day2.DL_RedRift_Day2" (select Loaded "Activated" "Unloaded") false :!class /Script/Engine.DataLayerManager))))

(event Custom|DC_Echo
  ;[0.6823 0.3546 0.6748 0.5] Begin Play / Calendar / Choose Orb location
  (Utilities|FlowControl|Sequence
    (:then_0
      (Utilities|FlowControl|Delay 1.0) ; wait for the data layer to turn on in PIE\n..probably not needed particularly because\nwe're using stateful event router events..\nbut can't hurt as PIE is goofy with slow\ndata layer loading sometimes.
      (GameplayEventRouter|BroadcastStatefulEvent (GameplayEventRouter|GetGlobalEventRouter) (Utilities|Struct|MakeDCNOrbCheatsEventRouter "NewEnumerator0") 0) ; Send to BGA_DCN_OrbActor so when it wakes\nup it sees this event and puts itself above\none of the Dark Voyager Echo bosses. This is\nusually calendar driven such as for 40.00 \nPower Hours.
    )
    (:then_1
      (bind _returnvalue (WorldPartition|GetDataLayerManager))
      (DataLayers|SetDataLayerRuntimeState _returnvalue "/DragonCartNarrative/DataLayers/DL_DCN_Orb_DV_Echo.DL_DCN_Orb_DV_Echo" "Activated" false :!class /Script/Engine.DataLayerManager) ; data layer: Orb used for Dark Voyager Echoes
    )))

(event Custom|Shard4Delivery
  ;[0.6823 0.3546 0.6748 0.5] Begin Play / Calendar / Choose Orb location
  ;; Delivery 4
  (CallFunction|OrbTechHQPhase 23) ; 3 shard pieces
)

(event Custom|Shard4Delivered
  ;[0.6823 0.3546 0.6748 0.5] Begin Play / Calendar / Choose Orb location
  ;; Delivery 4
  (CallFunction|OrbTechHQPhase 24) ; 3 shard pieces
)

(event Custom|4ShardsVisible
  ;[0.6823 0.3546 0.6748 0.5] Begin Play / Calendar / Choose Orb location
  ;; Shard Delivery 4+
  (CallFunction|OrbTechHQPhase 22) ; 4 shard pieces
)

(event Custom|5ShardsVisible
  ;[0.6823 0.3546 0.6748 0.5] Begin Play / Calendar / Choose Orb location
  ;; Shard Delivery 4+
  (CallFunction|OrbTechHQPhase 25) ; 5 shard pieces
)

(event Custom|6ShardsVisible
  ;[0.6823 0.3546 0.6748 0.5] Begin Play / Calendar / Choose Orb location
  ;; Shard Delivery 4+
  (CallFunction|OrbTechHQPhase 26) ; 6 shard pieces
)

(event Custom|7ShardsVisible
  ;[0.6823 0.3546 0.6748 0.5] Begin Play / Calendar / Choose Orb location
  ;; Shard Delivery 4+
  (CallFunction|OrbTechHQPhase 27) ; 6 shard pieces
)

(event Custom|Oasis_RiftGrow2
  ;[0.6823 0.3546 0.6748 0.5] Begin Play / Calendar / Choose Orb location
  ;; Oasis Events - 5th Shard Delivery
  (Variables|Default|SetSetRiftScale 0.7)
  (CallFunction|OrbTechHQPhase 28))

(event Custom|Oasis_RiftGrow1
  ;[0.6823 0.3546 0.6748 0.5] Begin Play / Calendar / Choose Orb location
  ;; Oasis Events - 5th Shard Delivery
  (Variables|Default|SetSetRiftScale 0.3)
  (CallFunction|OrbTechHQPhase 28))

(event Custom|Oasis_Delivery
  ;[0.6823 0.3546 0.6748 0.5] Begin Play / Calendar / Choose Orb location
  ;; Oasis Events - 5th Shard Delivery
  (CallFunction|OrbTechHQPhase 29))

(event Custom|Oasis_Delivery2
  ;[0.6823 0.3546 0.6748 0.5] Begin Play / Calendar / Choose Orb location
  ;; Oasis Events - 5th Shard Delivery
  (CallFunction|OrbTechHQPhase 30))

(event Custom|Oasis_Delivery3
  ;[0.6823 0.3546 0.6748 0.5] Begin Play / Calendar / Choose Orb location
  ;; Oasis Events - 5th Shard Delivery
  (CallFunction|OrbTechHQPhase 31)
  (Variables|Default|SetSetRiftScale 0.3)
  (CallFunction|OrbTechHQPhase 17))

(event Custom|Oasis_Delivered
  ;[0.6823 0.3546 0.6748 0.5] Begin Play / Calendar / Choose Orb location
  ;; Oasis Events - 5th Shard Delivery
  (CallFunction|OrbTechHQPhase 37) ; DC_5ShardsVisible should be on at the same time
  (Variables|Default|SetSetRiftScale 0.7)
  (CallFunction|OrbTechHQPhase 17))

(event Custom|Squid_Rift1
  ;[0.6823 0.3546 0.6748 0.5] Begin Play / Calendar / Choose Orb location
  ;; Squid Events - 6th Shard Delivery
  (Variables|Default|SetSetRiftScale 0.3)
  (CallFunction|OrbTechHQPhase 39))

(event Custom|Squid_Delivery
  ;[0.6823 0.3546 0.6748 0.5] Begin Play / Calendar / Choose Orb location
  ;; Squid Events - 6th Shard Delivery
  (CallFunction|OrbTechHQPhase 33))

(event Custom|Squid_Rift2
  ;[0.6823 0.3546 0.6748 0.5] Begin Play / Calendar / Choose Orb location
  ;; Squid Events - 6th Shard Delivery
  (Variables|Default|SetSetRiftScale 0.7)
  (CallFunction|OrbTechHQPhase 39))

(event Custom|Squid_Delivery2
  ;[0.6823 0.3546 0.6748 0.5] Begin Play / Calendar / Choose Orb location
  ;; Squid Events - 6th Shard Delivery
  (CallFunction|OrbTechHQPhase 34))

(event Custom|Squid_Delivered
  ;[0.6823 0.3546 0.6748 0.5] Begin Play / Calendar / Choose Orb location
  ;; Squid Events - 6th Shard Delivery
  (CallFunction|OrbTechHQPhase 36) ; DC_6ShardsVisible should be on at the same time
  (Variables|Default|SetSetRiftScale 0.7)
  (CallFunction|OrbTechHQPhase 17))

(event Custom|Squid_Delivery3
  ;[0.6823 0.3546 0.6748 0.5] Begin Play / Calendar / Choose Orb location
  ;; Squid Events - 6th Shard Delivery
  (CallFunction|OrbTechHQPhase 35)
  (Variables|Default|SetSetRiftScale 0.3)
  (CallFunction|OrbTechHQPhase 17))

(event Custom|OrbOcean
  ;[0.2316 0.3836 0.7396 0.5] Orb - Ocean data layers
  (bind _element (ArrayMacros|GetRandomArrayElement (Variables|Default|GetOceanOrbLayers))
    (:%20%20
      (Variables|Default|SetCurrentOrbLayer _element)
      (bind _returnvalue (WorldPartition|GetDataLayerManager))
      (DataLayers|SetDataLayerRuntimeState _returnvalue (Variables|Default|GetCurrentOrbLayer) "Activated" false :!class /Script/Engine.DataLayerManager)
      (for _array_element (Variables|Default|GetOceanOrbLayers)
        (if (!= (Variables|Default|GetCurrentOrbLayer) _array_element)
          (bind _returnvalue_2 (WorldPartition|GetDataLayerManager))
          (DataLayers|SetDataLayerRuntimeState _returnvalue_2 _array_element "Unloaded" false :!class /Script/Engine.DataLayerManager))))))

(event Custom|Enable%20Orb%20over%20Ocean
  ;[0.2316 0.3836 0.7396 0.5] Orb - Ocean data layers
  (bind _element (ArrayMacros|GetRandomArrayElement (Variables|Default|GetOceanOrbLayers))
    (:%20%20
      (Variables|Default|SetCurrentOrbLayer _element)
      (bind _returnvalue (WorldPartition|GetDataLayerManager))
      (DataLayers|SetDataLayerRuntimeState _returnvalue (Variables|Default|GetCurrentOrbLayer) "Activated" false :!class /Script/Engine.DataLayerManager)
      (for _array_element (Variables|Default|GetOceanOrbLayers)
        (if (!= (Variables|Default|GetCurrentOrbLayer) _array_element)
          (bind _returnvalue_2 (WorldPartition|GetDataLayerManager))
          (DataLayers|SetDataLayerRuntimeState _returnvalue_2 _array_element "Unloaded" false :!class /Script/Engine.DataLayerManager))))))

(event Custom|OrbTechHQPhase
  (params (Phase :type Integer))
  (bind _returnvalue (- Phase 1))
  (bind (_output_eventname_2_90f02b394aaeb20f36d7a8849eeccb09 _output_enableddatalayers_9_41bb9a3e4648805b5af70aa728288c92) (Utilities|Struct|BreakCalendarDataLayerData (Utilities|Array|Get(acopy) (Variables|Default|GetCalendarDataLayerList) _returnvalue)))
  (bind (_output_eventname_2_90f02b394aaeb20f36d7a8849eeccb09_2 _output_enableddatalayers_9_41bb9a3e4648805b5af70aa728288c92_2 _output_disableddatalayers_10_8de59e3b474c88ed2e7988b185572444) (Utilities|Struct|BreakCalendarDataLayerData (Utilities|Array|Get(acopy) (Variables|Default|GetCalendarDataLayerList) _returnvalue)))
  ;[0.5281 0.6875 0.5184 0.5] Orb - Tech HQ data layers
  (CallFunction|LogIt (Utilities|String|Append "OrbTechHQPhase: " (Utilities|String|ToString(Integer) Phase)) false)
  (bind _returnvalue_2 (WorldPartition|GetDataLayerManager))
  (Utilities|FlowControl|Sequence
    (:then_0
      (for _array_element _output_enableddatalayers_9_41bb9a3e4648805b5af70aa728288c92
        (CallFunction|LogIt (Utilities|String|Append "activating data layer: " (Utilities|GetDisplayName _array_element)) false) ; enable
        (DataLayers|SetDataLayerInstanceRuntimeState _returnvalue_2 (DataLayers|GetDataLayerInstancefromAsset _returnvalue_2 _array_element) "Activated" false) ; Activate  data layer
      ))
    (:then_1
      (for _array_element_2 _output_disableddatalayers_10_8de59e3b474c88ed2e7988b185572444
        (CallFunction|LogIt (Utilities|String|Append "unload data layer: " (Utilities|GetDisplayName _array_element_2)) false) ; unload
        (DataLayers|SetDataLayerInstanceRuntimeState _returnvalue_2 (DataLayers|GetDataLayerInstancefromAsset _returnvalue_2 _array_element_2) "Unloaded" false) ; Unload data layer
      ))) ; 0 - enable data layers\n1 - disable data layers
)

(fn LogIt
  (params (InString :type String) (bLogAsWarning :type Boolean))
  ; Log It
  (Fortnite|FortShippingLog self (Utilities|String|Append "[DCN Orb] BP_DCN_CalendarDataLayer - " InString) bLogAsWarning))

(fn OnRep_SetRiftScale
  (Networking|CanExecuteCosmeticEvents
    (:True
      (Rendering|Material|SetScalarParameterValue "/DragonCartNarrative/Assets/VFX/RiftPortalIdle/MPC_DragonCart_Orb_Rift.MPC_DragonCart_Orb_Rift" "RiftScale" (Variables|Default|GetSetRiftScale) :!class /Script/Engine.KismetMaterialLibrary))))

(fn ConstructionScript
  (|CallParentFunction))
