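; Server-driven controller for the DragonCart narrative Flying Orb, placing it according to calendar events once the match is ready.
; Event progress picks either a point on the ocean spline or a frame in the starter-island sequence, which a spawned replicated player briefly plays, then pauses.
; Orb transform and state broadcast through the gameplay event router for level injection.
; Cheat events adjust placement and turn or teleport players to the orb.
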(event EventBeginPlay
  ;[0.3248 0.4198 0.7031 0.5] Begin Play
  (if (Networking|IsServer)
    (Class|FortAsyncActionCurrentPlaylistReady|PlaylistDataReadyAsync
      (:OnReady
        (MCPContext|InitialCalendarSyncCompleted
          (:OnSyncComplete
            (Class|FortAsyncActionAuthorityMatchReady|MatchReadyAsync
              (:OnReady
                (CallFunction|CheckOrbCalendarEvents)))))))))

(event Custom|Check%20Orb%20Calendar%20Events
  ;[0.6805 0.394 0.7031 0.5] Calendar Events
  (Utilities|FlowControl|Sequence
    (:then_0
      (bind (_biseventactive _timesincebegin _timeuntilend _timespanratio) (CalendarEvents|GetCalendarEventInformation "FlyingOrb_Ocean")) ; loops a big circle around the island
      (if _biseventactive
        (CallFunction|OrbOceanCircle _timespanratio)))
    (:then_1
      (bind (_biseventactive_2 _timesincebegin_2 _timeuntilend_2 _timespanratio_2) (CalendarEvents|GetCalendarEventInformation "FlyingOrb_StarterIsland")) ; approaches the starter island, flies around it, and then\ngoes towards the main map where it disappears before\nentering the gameplay space
      (if _biseventactive_2
        (CallFunction|OrbStarterIsland _timespanratio_2)))) ; 0 - 39.00 Ocean Loop
)

(event Custom|Inject%20Level%20-%20Orb
  (params (Orb%20Data :type Event%20Router%20DCN%20Flying%20Orb%20Struct%20Structure))
  (bind (_transform_11_3049b11d41fcd0da0c3349a89986ffef _state_12_9b589b3a4af9b762cff3afaead22301e) (Utilities|Struct|BreakEventRouterDCNFlyingOrbStruct Orb%20Data))
  (bind (_transform_11_3049b11d41fcd0da0c3349a89986ffef_location _transform_11_3049b11d41fcd0da0c3349a89986ffef_rotation _transform_11_3049b11d41fcd0da0c3349a89986ffef_scale) (Math|Transform|BreakTransform _transform_11_3049b11d41fcd0da0c3349a89986ffef))
  ;[0.7031 0.3519 0.1499 0.5] Level Injection
  (Variables|DebugandCheats|SetLast%20Orb%20Spawn%20Location _transform_11_3049b11d41fcd0da0c3349a89986ffef_location) ; useful for cheats
  (GameplayEventRouter|BroadcastStatefulEvent (GameplayEventRouter|GetGlobalEventRouter) (Utilities|Struct|MakeEventRouterDCNFlyingOrbStruct (Math|Transform|MakeTransform _transform_11_3049b11d41fcd0da0c3349a89986ffef_location _transform_11_3049b11d41fcd0da0c3349a89986ffef_rotation (* _transform_11_3049b11d41fcd0da0c3349a89986ffef_scale (Math|Conversions|ToVector(Float) (ScalableFloat|GetValueatLevel (Variables|Default|GetScale%20Multiplier))))) _state_12_9b589b3a4af9b762cff3afaead22301e) "/DragonCartNarrative/Gameplay/FlyingOrb/EventRouter_DCN_FlyingOrb_Struct.EventRouter_DCN_FlyingOrb_Struct") ; Goes to BP_FortLevelInjectionManagerBase_DragonCart
)

(event Custom|Orb_TechHQ
  (params (percentage%20%280-1.0%29 :type Float%20%28double-precision%29))
  ;[0.4865 0.7031 0.5741 0.5] Cheats / Location specific commands
  (CallFunction|InjectLevel-Orb 0 (Utilities|Struct|MakeEventRouterDCNFlyingOrbStruct (Math|Transform|MakeTransform (CallFunction|Orb-TechHQLocation) (0.0 0.0 0.0) (Math|Conversions|ToVector(Float) (Variables|DebugAndCheats|GetConsole%20Cheat%20Scale))))))

(event Custom|Orb_ZOffset
  (params (ZOffset :type Float%20%28double-precision%29))
  ;[0.4865 0.7031 0.5741 0.5] Cheats / Location specific commands
  ;;[0.4522 0.5598 0.8594 0.5] Dev Cheats
  (Variables|DebugandCheats|SetConsole%20Cheat%20Z%20Offset ZOffset))

(event Custom|Orb_Scale
  (params (ZOffset :type Float%20%28double-precision%29))
  ;[0.4865 0.7031 0.5741 0.5] Cheats / Location specific commands
  ;;[0.4522 0.5598 0.8594 0.5] Dev Cheats
  (Variables|DebugAndCheats|SetConsole%20Cheat%20Scale ZOffset))

(event Custom|Orb_Rotate
  (params (Rotation :type Float%20%28double-precision%29))
  ;[0.4865 0.7031 0.5741 0.5] Cheats / Location specific commands
  ;;[0.4522 0.5598 0.8594 0.5] Dev Cheats
  (Variables|DebugandCheats|SetConsole%20Cheat%20Rotation Rotation))

(event Custom|Orb_OceanCircle
  (params (spline%20percentage%20%280%20-%201.0%29 :type Float%20%28double-precision%29))
  ;[0.4865 0.7031 0.5741 0.5] Cheats / Location specific commands
  (bind _world_location (CallFunction|GetSplineLocationAtPercentage spline%20percentage%20%280%20-%201.0%29 (Variables|Default|GetSpline%20-%20Ocean))) ; TODO: replace the spline with a sequence
  (CallFunction|InjectLevel-Orb 0 (Utilities|Struct|MakeEventRouterDCNFlyingOrbStruct (Math|Transform|MakeTransform _world_location))))

(event Custom|Inject%20Orb%20from%20Sequencer%20Actor
  ;[0.7031 0.3519 0.1499 0.5] Level Injection
  (bind _outactors (Actor|GetAllActorsOfClass "/DragonCartNarrative/Gameplay/FlyingOrb/BP_DCN_Orb_RootActor.BP_DCN_Orb_RootActor_C")) ; orb root actor
  (if (Utilities|Array|IsEmpty _outactors)
    (CallFunction|LogIt "Inject Orb from Root Actor - Didn't find the root actor!!" true)
    (else
      (CallFunction|LogIt "Inject Orb from Root Actor - Found the Orb Root Actor" true)
      (bind _output_get (Variables|Sequences|SetOrb%20Root%20Actor (Utilities|Array|Get(acopy) _outactors)))
      (CallFunction|InjectLevel-Orb 0 (Utilities|Struct|MakeEventRouterDCNFlyingOrbStruct (Transformation|GetActorTransform _output_get) (Variables|/DragonCartNarrative/Gameplay/FlyingOrb/BP_DCN_Orb_RootActor.BP_DCN_Orb_RootActor_C|GetOrb%20State _output_get))))))

(event Custom|Orb_LookAt
  ;[0.4865 0.7031 0.5741 0.5] Cheats / Location specific commands
  ;;[0.4522 0.5598 0.8594 0.5] Dev Cheats
  ;;; Make Player Face the Orb
  (if (>= (Math|Vector|VectorLength (Variables|DebugandCheats|GetLast%20Orb%20Spawn%20Location)) 1.0)
    (bind _outfortplayerpawns (Fortnite|GetFortPlayerPawns))
    (for _array_element _outfortplayerpawns
      (Transform|Actor|TeleportPlayerPawn _array_element (Transformation|GetActorLocation _array_element) (Math|Rotator|FindLookatRotation (Transformation|GetActorLocation _array_element) (Variables|DebugandCheats|GetLast%20Orb%20Spawn%20Location)) true true) ; rotate pawn to face the orb\n..probably don't want extra rotation values that take the\nplayer off the Z axis, but this node doesn't seem to mind
    )
    (else
      (Development|PrintString "Flying Orb - hasn't been spawned yet!!"))))

(event Custom|Orb_TeleportTo
  ;[0.4865 0.7031 0.5741 0.5] Cheats / Location specific commands
  ;;[0.4522 0.5598 0.8594 0.5] Dev Cheats
  ;;; Make Player Face the Orb_2
  (if (>= (Math|Vector|VectorLength (Variables|DebugandCheats|GetLast%20Orb%20Spawn%20Location)) 1.0)
    (bind _outfortplayerpawns (Fortnite|GetFortPlayerPawns))
    (for _array_element _outfortplayerpawns
      (Transform|Actor|TeleportPlayerPawn _array_element (Variables|DebugandCheats|GetLast%20Orb%20Spawn%20Location) (Math|Rotator|FindLookatRotation (Transformation|GetActorLocation _array_element) (Variables|DebugandCheats|GetLast%20Orb%20Spawn%20Location)) true true) ; teleport player pawn to the Orb
    )
    (else
      (Development|PrintString "Flying Orb - hasn't been spawned yet!!"))))

(event Custom|Orb_StarterIsland
  (params (percentage%20%280%20-%201.0%29 :type Float%20%28double-precision%29))
  ;[0.4865 0.7031 0.5741 0.5] Cheats / Location specific commands
  (Variables|SequencePlayer|SetSequence%20Percentage%20%280-1%29 percentage%20%280%20-%201.0%29)
  (Variables|DebugandCheats|SetExtra%20Debug%20String "Starter Island Approach")
  (Default|AssignLevelSequenceIniitialized (delegate Custom|Level%20Sequence%20Iniitialized_Event%20-%20Starter%20Island%20Approach)) ; Wait for Sequenecer to Play and Pause
  (CallFunction|%28Server%29Load,Play,AndQuicklyPauseSequence (Variables|Sequences|GetSequence%20-%20Starter%20Island%20Approach)))

(event Custom|Destroy%20Root%20Actors
  ;[0.7031 0.3519 0.1499 0.5] Level Injection
  (bind _outactors (Actor|GetAllActorsOfClass "/DragonCartNarrative/Gameplay/FlyingOrb/BP_DCN_Orb_RootActor.BP_DCN_Orb_RootActor_C")) ; orb root actor
  (for _array_element _outactors
    (Actor|DestroyActor _array_element) ; destroy all dynamically\nspawned root actors
  ))

(event Custom|Level%20Sequence%20Iniitialized_Event%20-%20Starter%20Island%20Approach
  ;[0.4865 0.7031 0.5741 0.5] Cheats / Location specific commands
  (Default|UnbindEventfromLevelSequenceIniitialized (delegate Custom|Level%20Sequence%20Iniitialized_Event%20-%20Starter%20Island%20Approach)))

(event Custom|%28server%29%20Load,%20Play,%20and%20Quickly%20Pause%20Sequence
  (params (Asset :type Soft%20Object%20Reference))
  ;[0.5938 0.584 0.2008 0.5] Replicated Level Sequence Player - Load, Play, Pause
  (Utilities|IsValid (Variables|SequencePlayer|GetSequence%20Player)
    (:Is%20Valid
      (Sequencer|Player|Stop (Variables|SequencePlayer|GetSequence%20Player) :!class /Script/MovieScene.MovieSceneSequencePlayer)
      (CallFunction|DestroyRootActors)
      (Actor|DestroyActor (Variables|SequencePlayer|GetSpawned%20Replicated%20Sequence%20Player)))
    (:Is%20Not%20Valid
      _)) ; sequencer player active?
  (CallFunction|LogIt "ASync Load Level Sequence Asset")
  (bind _object (Utilities|AsyncLoadAsset Asset)
    (:Completed
      (CallFunction|LogIt "Level Sequence Async loading finished!")
      (bind _aslevel_sequence (Utilities|Casting|CastToLevelSequence _object)
        (:then
          (bind _output_get (Variables|SequencePlayer|SetLevel%20Sequence%20-%20Loaded _aslevel_sequence))
          ;; Calculate Playback Range
          (bind _returnvalue (Sequencer|Sequence|GetPlaybackRange _output_get))
          (bind (_returnvalue_bhasstart _returnvalue_bhasend _returnvalue_inclusivestart _returnvalue_exclusiveend) (Utilities|Struct|BreakSequencerScriptingRange _returnvalue))
          (Variables|SequencePlayer|SetTotal%20Sequence%20Frames (- _returnvalue_exclusiveend _returnvalue_inclusivestart))
          (Variables|SequencePlayer|SetFirst%20Frame%20in%20Sequence _returnvalue_inclusivestart)
          (Variables|SequencePlayer|SetLast%20Frame%20in%20Sequence _returnvalue_exclusiveend)
          (bind _returnvalue_2 (Game|SpawnActorfromClass :Class "/Script/LevelSequence.ReplicatedLevelSequenceActor" :SpawnTransform (Math|Transform|MakeTransform "0.000000,0.000000,1000.000000") :CollisionHandlingOverride "AlwaysSpawn" :TransformScaleMethod "MultiplyWithRoot" :PlaybackSettings (Utilities|Struct|MakeMovieSceneSequencePlaybackSettings false "(Value=0)" (struct :TickIntervalSeconds 0.0 :EvaluationBudgetMicroseconds 0.0 :bTickWhenPaused false :bAllowRounding true) 1.0 0.0 false false false false false false "ForceKeepState" false) :LevelSequenceAsset (Variables|SequencePlayer|GetLevel%20Sequence%20-%20Loaded) :bReplicatePlayback true)) ; Spawn the sequence on the server, and allow replication to send the\ntime down to the client. Because clients may not have all actors loaded\nwhen this plays, it is incredibly important to design sections of the sequence\naround what will correctly replicate and init. These setups need to be\ndesigned around worst case of Join In Progress and the client not being\naround when the sequence runs.\n\nex: MPC values set in the sequence on the server should work for clients\nex: Toggling data layers should work correctly for hiding/unhiding things\nex: A spawnable may not show up, but an actor that lives in a data layer\n       and that uses MPCs to init its state should..?\n\nex: Setting a material param directly on a vfx may NOT work because that \nVFX may not be loaded on the client when the sequence runs. Better to \nuse MPC values.
          (Variables|SequencePlayer|SetSpawned%20Replicated%20Sequence%20Player _returnvalue_2)
          (bind _output_get_2 (Variables|SequencePlayer|SetSequence%20Player (Variables|/Script/LevelSequence.LevelSequenceActor|GetSequencePlayer _returnvalue_2)))
          (Sequencer|Player|SetPlaybackPosition _output_get_2 (Utilities|Struct|MakeMovieSceneSequencePlaybackParams (Utilities|Struct|MakeFrameTime (Utilities|Struct|MakeFrameNumber (+ (Variables|SequencePlayer|GetFirst%20Frame%20in%20Sequence) (Math|Float|Round (* (Math|Conversions|ToFloat(Integer) (Variables|SequencePlayer|GetTotal%20Sequence%20Frames)) (Variables|SequencePlayer|GetSequence%20Percentage%20%280-1%29)))))) 0.0 "" "(Hours=0,Minutes=0,Seconds=0,Frames=0,Subframe=0.000000,bDropFrameFormat=False)" "Frame" "Jump" true))
          (CallFunction|LogIt "Level Sequence about to Play/Pause")
          (Sequencer|Player|Play (Variables|SequencePlayer|GetSequence%20Player) :!class /Script/MovieScene.MovieSceneSequencePlayer)
          (Utilities|FlowControl|Delay 0.1) ; Let the sequence play for a moment\nbefore pausing.\n\nMaybe just need to wait one frame,\nor don't event need to wait at all\nbefore pausing?\n\n..need to test what happens if the\nsequence plays within 0.1 seconds of\nthe end and ends playback before\npause is called. Keep State should\nstill handle that scenario.
          (Sequencer|Player|Pause (Variables|SequencePlayer|GetSequence%20Player))
          (Default|CallLevelSequenceIniitialized) ; bound to inside this blueprint to\ntrigger Orb Level Injection and other\nmap location specific setup
        )
        (:CastFailed
          (CallFunction|LogIt "Cast to Level Sequence failed!" true))))) ; load the sequence
)

(fn GetSplineLocationatPercentage
  (params (percentage%20%280%20-%201.0%29 :type Float%20%28double-precision%29) (Spline%20Component :type Spline%20Component%20Object%20Reference))
  (return (Spline|GetLocationatDistanceAlongSpline Spline%20Component (* percentage%20%280%20-%201.0%29 (Spline|GetSplineLength Spline%20Component)) "World")))

(fn LogIt
  (params (string :type String) (bLogAsWarning :type Boolean))
  ; Log It
  (Fortnite|FortShippingLog 0 (Utilities|String|Append "[DragonCartNarrative FlyingOrb] " (Variables|DebugandCheats|GetExtra%20Debug%20String) " - " string) bLogAsWarning))

(fn Orb-GlobalZOffset
  (return (ScalableFloat|GetValueatLevel (Variables|TuningVariables|GetFlying%20Orb%20-%20Global%20Z%20Offset))))

(fn Orb-TechHQLocation
  (return (ScalableFloat|GetValueatLevel (Variables|TuningVariables|GetFlying%20Orb%20Location%20-%20Tech%20HQ)) (ScalableFloat|GetValueatLevel (Variables|TuningVariables|GetFlying%20Orb%20Location%20-%20Tech%20HQ) 1.0) (+ (ScalableFloat|GetValueatLevel (Variables|TuningVariables|GetFlying%20Orb%20Location%20-%20Tech%20HQ) 2.0) (CallFunction|Orb-TechHQZOffset))))

(fn Orb-TechHQZOffset
  (return (+ (CallFunction|Orb-GlobalZOffset) (Variables|DebugandCheats|GetConsole%20Cheat%20Z%20Offset))))

(fn ConstructionScript
  (Utilities|IsValid (Game|GetGameInstance)
    (:Is%20Not%20Valid
      (if (Editor|InEditororPIE)
        (Transformation|SetActorLocationAndRotation "-6144.000000,1024.000000,2000.000000") ; Please don't move this actor as it will move splines and all sorts\nof other things. Location tuning values are in the data tables.-Adam
      ))) ; Is game running?
)
