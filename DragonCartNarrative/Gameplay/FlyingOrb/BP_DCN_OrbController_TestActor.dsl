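; BP_DCN_OrbController_TestActor is a test actor for the flying orb controller in DragonCartNarrative gameplay.
; On BeginPlay it prints a debug string identifying itself as the Orb Controller test actor.
; The construction script is empty, and the file contains no orb logic.
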(event EventBeginPlay
  (Development|PrintString "BP_DCN_TestActor - Orb Controller"))

(fn ConstructionScript)
