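; Rift portal actor for the DragonCart narrative flying orb, sized by live calendar events.
; After playlist and calendar sync, active events set replicated UpdateMPC to zero (Rift Out) or a mapped scale, with matching audio attenuation.
; Its RepNotify drives the RiftScale material collection parameter off-server and scales the build-blocking collision, which a ScalableFloat can remove at startup.
; SetRiftSize and OrbTechHQPhase set size manually for testing.
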(event EventBeginPlay
  ; Set mpc setting based on calendar event
  (|CallParentFunction)
  (Class|FortAsyncActionCurrentPlaylistReady|PlaylistDataReadyAsync
    (:OnReady
      (MCPContext|InitialCalendarSyncCompleted
        (:OnSyncComplete
          (Utilities|FlowControl|Sequence
            (:then_0
              (bind _eventname (MCPContext|WatchCalendarEvent (Variables|Default|GetCalendar%20-%20Rift%20Out))
                (:OnEventActive
                  (Fortnite|FortShippingLog 0 (Utilities|String|Append _eventname "" "Is Active"))
                  (Variables|Default|SetUpdateMPC 0.0))))
            (:then_1
              (bind _eventname_2 (MCPContext|WatchCalendarEvents (Variables|Default|GetCalendar%20-%20RotatingOrbCalendars))
                (:OnEventActive
                  (Fortnite|FortShippingLog 0 (Utilities|String|Append _eventname_2 "" "Is Active"))
                  (Variables|Default|SetUpdateMPC (Utilities|Map|Find (Variables|Default|Get%20RotatingOrbCalendars%20Map) _eventname_2))
                  (CallFunction|UpdateAudioAttenuationSettings (Utilities|Map|Find (Variables|Default|GetOrbCalendarsAttenuationMap) _eventname_2)))))
            (:then_2
              (if (<= (ScalableFloat|GetValueatLevel (Variables|Default|GetHF_Rift_EnableBlockBuild)) 0)
                (Components|DestroyComponent (Variables|Default|GetBlockBuildingCollision))))))))))

(event Custom|SetRiftSize
  (params (UpdateMPC :type Float%20%28double-precision%29))
  ; Testing Only
  (Variables|Default|SetUpdateMPC UpdateMPC))

(event Custom|OrbTechHQPhase
  (params (Phase :type Integer))
  (bind _riftoutphase (Variables|Default|GetRiftoutPhase))
  (bind _returnvalue (- Phase _riftoutphase))
  (bind (_key _value) (Utilities|Map|GetKeyValueByIndex (Variables|Default|GetOrbCalendarsAttenuationMap) _returnvalue))
  (bind (_key_2 _value_2) (Utilities|Map|GetKeyValueByIndex (Variables|Default|Get%20RotatingOrbCalendars%20Map) _returnvalue))
  ; With phases Testing Only
  (Variables|Default|SetUpdateMPC _value_2)
  (CallFunction|UpdateAudioAttenuationSettings _value))

(event Custom|UpdateAudioAttenuationSettings
  (params (Value :type Float%20%28double-precision%29))
  (bind _riftloopaudio (Variables|Default|GetRiftLoopAudio))
  (bind _outattenuationsettings (Audio|Components|Audio|GetAttenuationSettingsToApply _riftloopaudio))
  (bind _output_get (Variables|/Script/Engine.AudioComponent|SetAttenuationOverrides _outattenuationsettings _riftloopaudio))
  (Variables|/Script/Engine.AudioComponent|SetbOverrideAttenuation true _riftloopaudio)
  (bind _structout (Utilities|Struct|SetmembersinSoundAttenuationSettings _output_get Value))
  (Audio|Components|Audio|AdjustAttenuation _riftloopaudio _structout))

(fn OnRep_UpdateMPC
  (bind _updatempc (Variables|Default|GetUpdateMPC))
  (if (not (Networking|IsServer))
    (Rendering|Material|SetScalarParameterValue "/DragonCartNarrative/Assets/VFX/RiftPortalIdle/MPC_DragonCart_Orb_Rift.MPC_DragonCart_Orb_Rift" "RiftScale" _updatempc :!class /Script/Engine.KismetMaterialLibrary))
  (Transformation|SetWorldScale3D (Variables|Default|GetBlockBuildingCollision) (Math|Conversions|ToVector(Float) _updatempc)))

(fn ConstructionScript)
