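; Server-side level injection manager for DragonCart that swaps the injected Flying Orb level between open and closed variants.
; On BeginPlay it waits for playlist data and match readiness, then listens on the global event router for the stateful orb struct sent by BP_DCN_FlyingOrbController.
; It removes the previously injected level, async-loads the injection data chosen by the payload's state, requests injection and stores the returned GUID.
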(event EventBeginPlay
  ;[0.4948 0.5208 0.3775 0.5] Begin Play / Event Router / Inject Level
  (if (Networking|IsServer)
    (Class|FortAsyncActionCurrentPlaylistReady|PlaylistDataReadyAsync
      (:OnReady
        (Class|FortAsyncActionAuthorityMatchReady|MatchReadyAsync
          (:OnReady
            (bind (_asynctaskproxy _payload) (GameplayEventRouter|StartListeningToStatefulEvent (GameplayEventRouter|GetGlobalEventRouter) "/DragonCartNarrative/Gameplay/FlyingOrb/EventRouter_DCN_FlyingOrb_Struct.EventRouter_DCN_FlyingOrb_Struct")
              (:OnEventReceived
                _)
              (:OnSavedEventStateExists
                _)) ; comes from BP_DCN_FlyingOrbController
            (CallFunction|RemoveInjectedLevel (Variables|LevelInjection|GetInjected%20Level%20GUID))
            (bind _object (Utilities|AsyncLoadAsset (Utilities|Select :Index (.State_12_9B589B3A4AF9B762CFF3AFAEAD22301E (Utilities|Struct|BreakEventRouterDCNFlyingOrbStruct _payload)) :NewEnumerator0 (Variables|LevelInjection|GetOpen%20Orb%20Level%20Injection%20Data) :NewEnumerator1 (Variables|LevelInjection|GetClosed%20Orb%20Level%20Injection%20Data)))
              (:Completed
                (bind _asfort_level_injection_data_base (Utilities|Casting|CastToFortLevelInjectionDataBase _object))
                (bind _injectedlevelguid (CallFunction|RequesttoInjectLevel (Utilities|ResolveSoftReference _asfort_level_injection_data_base) (Utilities|Struct|BreakEventRouterDCNFlyingOrbStruct _payload)))
                (Variables|LevelInjection|SetInjected%20Level%20GUID _injectedlevelguid)))))))))
